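# Physics asset for the right leg mesh of the F_MED_Virtuous player cosmetic body, defining secondary-motion chains on capsule bodies.
# A three-link leg_feat chain hangs from a master body on the thigh, while a tassel on the calf branches from one body into two tips.
# Small master capsules and the thigh-calf joint have free angular motion; chain links use narrow swing and very tight twist limits.
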PhysicsAsset /BRCosmetics/Characters/Player/Female/Medium/Bodies/F_MED_Virtuous/Meshes/F_MED_Virtuous_leg_r_Physics
bodies {
 calf_r [{shapeType=Capsule, length=30, radius=8, shapeName=calf_r_capsule, center=<25 3 3>, rotation=<rot -80 -0.0001 10>}]
 dyn_r_leg_feat_01 [{shapeType=Capsule, length=8, radius=5, shapeName=dyn_r_leg_feat_01_capsule, center=<12 0 -2.5>, rotation=<rot 90 0 0>}]
 dyn_r_leg_feat_02 [{shapeType=Capsule, length=15, radius=5, shapeName=dyn_r_leg_feat_02_capsule, center=<7 0 -2.5>, rotation=<rot 90 0 0>}]
 dyn_r_leg_feat_03 [{shapeType=Capsule, length=15, radius=5, shapeName=dyn_r_leg_feat_03_capsule, center=<7 0 0>, rotation=<rot 90 0 0>}]
 dyn_r_leg_feat_MSTR [{shapeType=Capsule, length=0.505, radius=0.505, shapeName=dyn_r_leg_feat_MSTR_capsule, rotation=<rot 0 -90 0>}]
 dyn_tassle_r [{shapeType=Capsule, length=2, radius=4, shapeName=dyn_tassle_r_capsule, center=<1 0 -3>, rotation=<rot 74.9999 180 179.9999>}]
 dyn_tassle_r_01 [{shapeType=Capsule, length=6, radius=5, shapeName=dyn_tassle_r_01_capsule, center=<6 0 -3>, rotation=<rot 90 0 0>}]
 dyn_tassle_r_02 [{shapeType=Capsule, length=6, radius=5, shapeName=dyn_tassle_r_02_capsule, center=<6 0 -3>, rotation=<rot 90 0 0>}]
 dyn_tassle_r_MSTR [{shapeType=Capsule, length=0.505, radius=0.505, shapeName=dyn_tassle_r_MSTR_capsule, rotation=<rot 0 -90 0>}]
 thigh_r [{shapeType=Capsule, length=20, radius=6, shapeName=thigh_r_capsule, center=<30 0 -1>, rotation=<rot 90 0 0>}]
}
constraints [
 {bone1Name=dyn_r_leg_feat_03, bone2Name=dyn_r_leg_feat_02, swing1LimitDegrees=18, swing2LimitDegrees=18, twistLimitDegrees=2}
 {bone1Name=dyn_r_leg_feat_01, bone2Name=dyn_r_leg_feat_MSTR, swing1LimitDegrees=18, swing2LimitDegrees=18, twistLimitDegrees=2}
 {bone1Name=dyn_r_leg_feat_02, bone2Name=dyn_r_leg_feat_01, swing1LimitDegrees=18, swing2LimitDegrees=18, twistLimitDegrees=2}
 {
  bone1Name=calf_r
  bone2Name=thigh_r
  swing1LimitDegrees=45
  swing1Motion=Free
  swing2LimitDegrees=45
  swing2Motion=Free
  twistLimitDegrees=45
  twistMotion=Free
 }
 {
  bone1Name=dyn_r_leg_feat_MSTR
  bone2Name=thigh_r
  swing1LimitDegrees=45
  swing1Motion=Free
  swing2LimitDegrees=45
  swing2Motion=Free
  twistLimitDegrees=45
  twistMotion=Free
 }
 {bone1Name=dyn_tassle_r_02, bone2Name=dyn_tassle_r, swing1LimitDegrees=12, swing2LimitDegrees=24, twistLimitDegrees=2}
 {bone1Name=dyn_tassle_r_01, bone2Name=dyn_tassle_r, swing1LimitDegrees=12, swing2LimitDegrees=24, twistLimitDegrees=2}
 {bone1Name=dyn_tassle_r, bone2Name=dyn_tassle_r_MSTR, swing1LimitDegrees=6, swing2LimitDegrees=6, twistLimitDegrees=1}
 {
  bone1Name=dyn_tassle_r_MSTR
  bone2Name=calf_r
  swing1LimitDegrees=45
  swing1Motion=Free
  swing2LimitDegrees=45
  swing2Motion=Free
  twistLimitDegrees=45
  twistMotion=Free
 }
]
