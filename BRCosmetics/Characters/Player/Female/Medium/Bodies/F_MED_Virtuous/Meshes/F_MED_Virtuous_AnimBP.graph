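# AnimBlueprint for the F_MED_Virtuous female cosmetic body, adding secondary motion to leg pieces, cuffs and skirt.
# A cached pose feeds standing and crouch Control Rig branches, each simulating both legs with calf-relative RigidBody nodes.
# The crouch branch adds calf rotation constraints and stronger gravity, and bIsCrouching selects between branches.
# The blended pose then runs through chained per-bone AnimDynamics nodes for cuffs and pelvis-relative, spring-driven skirt bones.
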AnimBlueprint /BRCosmetics/Characters/Player/Female/Medium/Bodies/F_MED_Virtuous/Meshes/F_MED_Virtuous_AnimBP
target_skeleton=&/F_MED_Virtuous_Skeleton
graphs [
 AnimGraph {nodes [
   n1 = Control Rig(Source=Use cached pose 'SavedPose12245'()) {
    alphaBoolBlend {blendOption=Linear}
    controlRigAssetReference {blueprintRigClass=&/F_MED_Virtuous_CtrlRig.F_MED_Virtuous_CtrlRig_C}
   }
   n2 = RigidBody(ComponentPose=Local To Component(LocalPose=@n1)) {
    bOverrideWorldGravity=true
    overridePhysicsAsset=&/F_MED_Virtuous_leg_l_Physics
    simulationSpace=BaseBoneSpace
    alphaBoolBlend {blendOption=Linear}
    baseBoneRef {boneName=calf_l}
    componentAppliedLinearAccClamp=<3000 3000 3000>
    componentLinearAccScale=<0.2 0.2 0.4>
    componentLinearVelScale=<0.1 0.1 0.2>
    overrideWorldGravity {z=-240}
    simSpaceSettings {
     maxAngularAcceleration=1000
     maxAngularVelocity=3
     maxLinearAcceleration=700
     maxLinearVelocity=500
     velocityScaleZ=0
     worldAlpha=1
     externalLinearDragV=<0.2 0.2>
    }
   }
   n3 = RigidBody(ComponentPose=@n2) {
    bOverrideWorldGravity=true
    overridePhysicsAsset=&/F_MED_Virtuous_leg_r_Physics
    simulationSpace=BaseBoneSpace
    alphaBoolBlend {blendOption=Linear}
    baseBoneRef {boneName=calf_r}
    componentAppliedLinearAccClamp=<3000 3000 3000>
    componentLinearAccScale=<0.2 0.2 0.4>
    componentLinearVelScale=<0.1 0.1 0.2>
    overrideWorldGravity {z=-240}
    simSpaceSettings {
     maxAngularAcceleration=1000
     maxAngularVelocity=3
     maxLinearAcceleration=700
     maxLinearVelocity=500
     velocityScaleZ=0
     worldAlpha=1
     externalLinearDragV=<0.2 0.2>
    }
   }
   n4 = Component To Local(ComponentPose=@n3)
   n5 = Control Rig(Source=Use cached pose 'SavedPose12245'()) {
    alphaBoolBlend {blendOption=Linear}
    controlRigAssetReference {blueprintRigClass=&/F_MED_Virtuous_Crouch_CtrlRig.F_MED_Virtuous_Crouch_CtrlRig_C}
   }
   n6 = "Constraint - dyn_l_leg_feat_MSTR (R)"(ComponentPose=Local To Component(LocalPose=@n5), Alpha=0.9, ConstraintWeights_0=1) {
    alphaBoolBlend {blendOption=Linear}
    boneToModify {boneName=dyn_l_leg_feat_MSTR}
    constraintSetup [{transformType=Rotation, targetBone {boneName=calf_l}}]
    constraintWeights [1]
   }
   n7 = "Constraint - dyn_l_leg_feat_02 (R)"(ComponentPose=@n6, Alpha=0.8, ConstraintWeights_0=1) {
    alphaBoolBlend {blendOption=Linear}
    boneToModify {boneName=dyn_l_leg_feat_02}
    constraintSetup [{transformType=Rotation, targetBone {boneName=calf_l}}]
    constraintWeights [1]
   }
   n8 = "Constraint - dyn_r_leg_feat_MSTR (R)"(ComponentPose=@n7, Alpha=0.9, ConstraintWeights_0=1) {
    alphaBoolBlend {blendOption=Linear}
    boneToModify {boneName=dyn_r_leg_feat_MSTR}
    constraintSetup [{transformType=Rotation, targetBone {boneName=calf_r}}]
    constraintWeights [1]
   }
   n9 = "Constraint - dyn_r_leg_feat_02 (R)"(ComponentPose=@n8, Alpha=0.8, ConstraintWeights_0=1) {
    alphaBoolBlend {blendOption=Linear}
    boneToModify {boneName=dyn_r_leg_feat_02}
    constraintSetup [{transformType=Rotation, targetBone {boneName=calf_r}}]
    constraintWeights [1]
   }
   n10 = RigidBody(ComponentPose=@n9) {
    bOverrideWorldGravity=true
    overridePhysicsAsset=&F_MED_Virtuous_leg_l_Physics
    simulationSpace=BaseBoneSpace
    alphaBoolBlend {blendOption=Linear}
    baseBoneRef {boneName=calf_l}
    componentAppliedLinearAccClamp=<3000 3000 3000>
    componentLinearAccScale=<0.2 0.2 0.4>
    componentLinearVelScale=<0.1 0.1 0.2>
    overrideWorldGravity {z=-980}
    simSpaceSettings {
     maxAngularAcceleration=1000
     maxAngularVelocity=3
     maxLinearAcceleration=700
     maxLinearVelocity=500
     velocityScaleZ=0
     worldAlpha=1
     externalLinearDragV=<0.2 0.2>
    }
   }
   n11 = RigidBody(ComponentPose=@n10) {
    bOverrideWorldGravity=true
    overridePhysicsAsset=&F_MED_Virtuous_leg_r_Physics
    simulationSpace=BaseBoneSpace
    alphaBoolBlend {blendOption=Linear}
    baseBoneRef {boneName=calf_r}
    componentAppliedLinearAccClamp=<3000 3000 3000>
    componentLinearAccScale=<0.2 0.2 0.4>
    componentLinearVelScale=<0.1 0.1 0.2>
    overrideWorldGravity {z=-980}
    simSpaceSettings {
     maxAngularAcceleration=1000
     maxAngularVelocity=3
     maxLinearAcceleration=700
     maxLinearVelocity=500
     velocityScaleZ=0
     worldAlpha=1
     externalLinearDragV=<0.2 0.2>
    }
   }
   n12 = Component To Local(ComponentPose=@n11)
   n13 = Get bIsCrouching()
   n14 = Anim Dynamics - Bone: dyn_cuff_01_01(
    ComponentPose=Local To Component(LocalPose=Blend Poses by bool(BlendPose_0=@n12, BlendPose_1=@n4, bActiveValue=@n13) {blendType=Linear})
    Alpha=1
    ExternalForce="0,0,0"
   ) {
    bUseGravityOverride=true
    simulationSpace=RootRelative
    alphaBoolBlend {blendOption=Linear}
    boundBone {boneName=dyn_cuff_01_01}
    componentLinearAccScale=<1 1 1>
    componentLinearVelScale=<1 1 1>
    gravityOverride {z=-60}
    physicsBodyDefinitions [{
      boundBone {boneName=dyn_cuff_01_01}
      boxExtents=<9 12 12>
      constraintSetup {angularLimitsMax=<2 2 2>, angularLimitsMin=<-2 -5 -2>, angularTarget=<0 0 0>}
      localJointOffset=<2 0 0>
     }]
    relativeSpaceBone {boneName=dyn_cuff_MSTR}
   }
   n15 = Anim Dynamics - Bone: dyn_cuff_01_02(ComponentPose=@n14, Alpha=1, ExternalForce="0,0,0") {
    bUseGravityOverride=true
    simulationSpace=RootRelative
    alphaBoolBlend {blendOption=Linear}
    boundBone {boneName=dyn_cuff_01_02}
    componentLinearAccScale=<1 1 1>
    componentLinearVelScale=<1 1 1>
    gravityOverride {z=-60}
    physicsBodyDefinitions [{
      boundBone {boneName=dyn_cuff_01_02}
      boxExtents=<9 12 12>
      constraintSetup {angularLimitsMax=<2 2 2>, angularLimitsMin=<-2 -5 -2>, angularTarget=<0 0 0>}
      localJointOffset=<2 0 0>
     }]
    relativeSpaceBone {boneName=dyn_cuff_MSTR}
   }
   n16 = Anim Dynamics - Bone: dyn_cuff_02_01(ComponentPose=@n15, Alpha=1, ExternalForce="0,0,0") {
    bUseGravityOverride=true
    simulationSpace=RootRelative
    alphaBoolBlend {blendOption=Linear}
    boundBone {boneName=dyn_cuff_02_01}
    componentLinearAccScale=<1 1 1>
    componentLinearVelScale=<1 1 1>
    gravityOverride {z=-60}
    physicsBodyDefinitions [{
      boundBone {boneName=dyn_cuff_02_01}
      boxExtents=<9 12 12>
      constraintSetup {angularLimitsMax=<2 2 2>, angularLimitsMin=<-2 -5 -2>, angularTarget=<0 0 0>}
      localJointOffset=<2 0 0>
     }]
    relativeSpaceBone {boneName=dyn_cuff_MSTR}
   }
   n17 = Anim Dynamics - Bone: dyn_cuff_02_02(ComponentPose=@n16, Alpha=1, ExternalForce="0,0,0") {
    bUseGravityOverride=true
    simulationSpace=RootRelative
    alphaBoolBlend {blendOption=Linear}
    boundBone {boneName=dyn_cuff_02_02}
    componentLinearAccScale=<1 1 1>
    componentLinearVelScale=<1 1 1>
    gravityOverride {z=-60}
    physicsBodyDefinitions [{
      boundBone {boneName=dyn_cuff_02_02}
      boxExtents=<9 12 12>
      constraintSetup {angularLimitsMax=<2 2 2>, angularLimitsMin=<-2 -5 -2>, angularTarget=<0 0 0>}
      localJointOffset=<2 0 0>
     }]
    relativeSpaceBone {boneName=dyn_cuff_MSTR}
   }
   n18 = Anim Dynamics - Bone: dyn_cuff_03_01(ComponentPose=@n17, Alpha=1, ExternalForce="0,0,0") {
    bUseGravityOverride=true
    simulationSpace=RootRelative
    alphaBoolBlend {blendOption=Linear}
    boundBone {boneName=dyn_cuff_03_01}
    componentLinearAccScale=<1 1 1>
    componentLinearVelScale=<1 1 1>
    gravityOverride {z=-60}
    physicsBodyDefinitions [{
      boundBone {boneName=dyn_cuff_03_01}
      boxExtents=<9 12 12>
      constraintSetup {angularLimitsMax=<2 2 2>, angularLimitsMin=<-2 -5 -2>, angularTarget=<0 0 0>}
      localJointOffset=<2 0 0>
     }]
    relativeSpaceBone {boneName=dyn_cuff_MSTR}
   }
   n19 = Anim Dynamics - Bone: dyn_cuff_03_02(ComponentPose=@n18, Alpha=1, ExternalForce="0,0,0") {
    bUseGravityOverride=true
    simulationSpace=RootRelative
    alphaBoolBlend {blendOption=Linear}
    boundBone {boneName=dyn_cuff_03_02}
    componentLinearAccScale=<1 1 1>
    componentLinearVelScale=<1 1 1>
    gravityOverride {z=-60}
    physicsBodyDefinitions [{
      boundBone {boneName=dyn_cuff_03_02}
      boxExtents=<9 12 12>
      constraintSetup {angularLimitsMax=<2 2 2>, angularLimitsMin=<-2 -5 -2>, angularTarget=<0 0 0>}
      localJointOffset=<2 0 0>
     }]
    relativeSpaceBone {boneName=dyn_cuff_MSTR}
   }
   n20 = Anim Dynamics - Bone: dyn_skirt_01(ComponentPose=@n19, Alpha=1, ExternalForce="0,0,0") {
    angularSpringConstant=240
    bAngularSpring=true
    bUseGravityOverride=true
    simulationSpace=RootRelative
    alphaBoolBlend {blendOption=Linear}
    boundBone {boneName=dyn_skirt_01}
    componentAppliedLinearAccClamp=<3000 3000 3000>
    componentLinearAccScale=<0.1 0.1 0.3>
    componentLinearVelScale=<0.1 0.1 0.1>
    gravityOverride {z=-360}
    physicsBodyDefinitions [{
      boundBone {boneName=dyn_skirt_01}
      boxExtents=<24 32 32>
      constraintSetup {angularLimitsMax=<1 2 6>, angularLimitsMin=<-1 -20 -6>, angularTarget=<1 0 0>}
      localJointOffset=<12 0 0>
     }]
    relativeSpaceBone {boneName=pelvis}
   }
   n21 = Anim Dynamics - Bone: dyn_skirt_01_mid(ComponentPose=@n20, Alpha=1, ExternalForce="0,0,0") {
    angularSpringConstant=240
    bAngularSpring=true
    bUseGravityOverride=true
    simulationSpace=RootRelative
    alphaBoolBlend {blendOption=Linear}
    boundBone {boneName=dyn_skirt_01_mid}
    componentAppliedLinearAccClamp=<3000 3000 3000>
    componentLinearAccScale=<0.1 0.1 0.3>
    componentLinearVelScale=<0.1 0.1 0.1>
    gravityOverride {z=-360}
    physicsBodyDefinitions [{
      boundBone {boneName=dyn_skirt_01_mid}
      boxExtents=<24 32 32>
      constraintSetup {angularLimitsMax=<1 2 6>, angularLimitsMin=<-1 -20 -6>, angularTarget=<1 0 0>}
      localJointOffset=<12 0 0>
     }]
    relativeSpaceBone {boneName=pelvis}
   }
   n22 = Anim Dynamics - Bone: dyn_skirt_01_a(ComponentPose=@n21, Alpha=1, ExternalForce="0,0,0") {
    angularSpringConstant=240
    bAngularSpring=true
    bUseGravityOverride=true
    simulationSpace=RootRelative
    alphaBoolBlend {blendOption=Linear}
    boundBone {boneName=dyn_skirt_01_a}
    componentAppliedLinearAccClamp=<3000 3000 3000>
    componentLinearAccScale=<0.1 0.1 0.3>
    componentLinearVelScale=<0.1 0.1 0.1>
    gravityOverride {z=-360}
    physicsBodyDefinitions [{
      boundBone {boneName=dyn_skirt_01_a}
      boxExtents=<24 32 32>
      constraintSetup {angularLimitsMax=<1 2 6>, angularLimitsMin=<-1 -20 -6>, angularTarget=<1 0 0>}
      localJointOffset=<12 0 0>
     }]
    relativeSpaceBone {boneName=pelvis}
   }
   n23 = Anim Dynamics - Bone: dyn_skirt_01_a_tip(ComponentPose=@n22, Alpha=1, ExternalForce="0,0,0") {
    angularSpringConstant=240
    bAngularSpring=true
    bUseGravityOverride=true
    simulationSpace=RootRelative
    alphaBoolBlend {blendOption=Linear}
    boundBone {boneName=dyn_skirt_01_a_tip}
    componentAppliedLinearAccClamp=<3000 3000 3000>
    componentLinearAccScale=<0.1 0.1 0.3>
    componentLinearVelScale=<0.1 0.1 0.1>
    gravityOverride {z=-360}
    physicsBodyDefinitions [{
      boundBone {boneName=dyn_skirt_01_a_tip}
      boxExtents=<24 32 32>
      constraintSetup {angularLimitsMax=<1 2 6>, angularLimitsMin=<-1 -20 -6>, angularTarget=<1 0 0>}
      localJointOffset=<12 0 0>
     }]
    relativeSpaceBone {boneName=pelvis}
   }
   n24 = Anim Dynamics - Bone: dyn_skirt_01_b(ComponentPose=@n23, Alpha=1, ExternalForce="0,0,0") {
    angularSpringConstant=240
    bAngularSpring=true
    bUseGravityOverride=true
    simulationSpace=RootRelative
    alphaBoolBlend {blendOption=Linear}
    boundBone {boneName=dyn_skirt_01_b}
    componentAppliedLinearAccClamp=<3000 3000 3000>
    componentLinearAccScale=<0.1 0.1 0.3>
    componentLinearVelScale=<0.1 0.1 0.1>
    gravityOverride {z=-360}
    physicsBodyDefinitions [{
      boundBone {boneName=dyn_skirt_01_b}
      boxExtents=<24 32 32>
      constraintSetup {angularLimitsMax=<1 2 6>, angularLimitsMin=<-1 -20 -6>, angularTarget=<1 0 0>}
      localJointOffset=<12 0 0>
     }]
    relativeSpaceBone {boneName=pelvis}
   }
   n25 = Anim Dynamics - Bone: dyn_skirt_01_b_tip(ComponentPose=@n24, Alpha=1, ExternalForce="0,0,0") {
    angularSpringConstant=240
    bAngularSpring=true
    bUseGravityOverride=true
    simulationSpace=RootRelative
    alphaBoolBlend {blendOption=Linear}
    boundBone {boneName=dyn_skirt_01_b_tip}
    componentAppliedLinearAccClamp=<3000 3000 3000>
    componentLinearAccScale=<0.1 0.1 0.3>
    componentLinearVelScale=<0.1 0.1 0.1>
    gravityOverride {z=-360}
    physicsBodyDefinitions [{
      boundBone {boneName=dyn_skirt_01_b_tip}
      boxExtents=<24 32 32>
      constraintSetup {angularLimitsMax=<1 2 6>, angularLimitsMin=<-1 -20 -6>, angularTarget=<1 0 0>}
      localJointOffset=<12 0 0>
     }]
    relativeSpaceBone {boneName=pelvis}
   }
   n26 = Anim Dynamics - Bone: dyn_skirt_01_c(ComponentPose=@n25, Alpha=1, ExternalForce="0,0,0") {
    angularSpringConstant=240
    bAngularSpring=true
    bUseGravityOverride=true
    simulationSpace=RootRelative
    alphaBoolBlend {blendOption=Linear}
    boundBone {boneName=dyn_skirt_01_c}
    componentAppliedLinearAccClamp=<3000 3000 3000>
    componentLinearAccScale=<0.1 0.1 0.3>
    componentLinearVelScale=<0.1 0.1 0.1>
    gravityOverride {z=-360}
    physicsBodyDefinitions [{
      boundBone {boneName=dyn_skirt_01_c}
      boxExtents=<24 32 32>
      constraintSetup {angularLimitsMax=<1 2 6>, angularLimitsMin=<-1 -20 -6>, angularTarget=<1 0 0>}
      localJointOffset=<12 0 0>
     }]
    relativeSpaceBone {boneName=pelvis}
   }
   n27 = Anim Dynamics - Bone: dyn_skirt_01_c_tip(ComponentPose=@n26, Alpha=1, ExternalForce="0,0,0") {
    angularSpringConstant=240
    bAngularSpring=true
    bUseGravityOverride=true
    simulationSpace=RootRelative
    alphaBoolBlend {blendOption=Linear}
    boundBone {boneName=dyn_skirt_01_c_tip}
    componentAppliedLinearAccClamp=<3000 3000 3000>
    componentLinearAccScale=<0.1 0.1 0.3>
    componentLinearVelScale=<0.1 0.1 0.1>
    gravityOverride {z=-360}
    physicsBodyDefinitions [{
      boundBone {boneName=dyn_skirt_01_c_tip}
      boxExtents=<24 32 32>
      constraintSetup {angularLimitsMax=<1 2 6>, angularLimitsMin=<-1 -20 -6>, angularTarget=<1 0 0>}
      localJointOffset=<12 0 0>
     }]
    relativeSpaceBone {boneName=pelvis}
   }
   n28 = Anim Dynamics - Bone: dyn_skirt_02(ComponentPose=@n27, Alpha=1, ExternalForce="0,0,0") {
    angularSpringConstant=240
    bAngularSpring=true
    bUseGravityOverride=true
    simulationSpace=RootRelative
    alphaBoolBlend {blendOption=Linear}
    boundBone {boneName=dyn_skirt_02}
    componentAppliedLinearAccClamp=<3000 3000 3000>
    componentLinearAccScale=<0.1 0.1 0.3>
    componentLinearVelScale=<0.1 0.1 0.1>
    gravityOverride {z=-360}
    physicsBodyDefinitions [{
      boundBone {boneName=dyn_skirt_02}
      boxExtents=<24 32 32>
      constraintSetup {angularLimitsMax=<1 2 6>, angularLimitsMin=<-1 -20 -6>, angularTarget=<1 0 0>}
      localJointOffset=<12 0 0>
     }]
    relativeSpaceBone {boneName=pelvis}
   }
   n29 = Anim Dynamics - Bone: dyn_skirt_02_mid(ComponentPose=@n28, Alpha=1, ExternalForce="0,0,0") {
    angularSpringConstant=240
    bAngularSpring=true
    bUseGravityOverride=true
    simulationSpace=RootRelative
    alphaBoolBlend {blendOption=Linear}
    boundBone {boneName=dyn_skirt_02_mid}
    componentAppliedLinearAccClamp=<3000 3000 3000>
    componentLinearAccScale=<0.1 0.1 0.3>
    componentLinearVelScale=<0.1 0.1 0.1>
    gravityOverride {z=-360}
    physicsBodyDefinitions [{
      boundBone {boneName=dyn_skirt_02_mid}
      boxExtents=<24 32 32>
      constraintSetup {angularLimitsMax=<1 2 6>, angularLimitsMin=<-1 -20 -6>, angularTarget=<1 0 0>}
      localJointOffset=<12 0 0>
     }]
    relativeSpaceBone {boneName=pelvis}
   }
   n30 = Anim Dynamics - Bone: dyn_skirt_02_a(ComponentPose=@n29, Alpha=1, ExternalForce="0,0,0") {
    angularSpringConstant=240
    bAngularSpring=true
    bUseGravityOverride=true
    simulationSpace=RootRelative
    alphaBoolBlend {blendOption=Linear}
    boundBone {boneName=dyn_skirt_02_a}
    componentAppliedLinearAccClamp=<3000 3000 3000>
    componentLinearAccScale=<0.1 0.1 0.3>
    componentLinearVelScale=<0.1 0.1 0.1>
    gravityOverride {z=-360}
    physicsBodyDefinitions [{
      boundBone {boneName=dyn_skirt_02_a}
      boxExtents=<24 32 32>
      constraintSetup {angularLimitsMax=<1 2 6>, angularLimitsMin=<-1 -20 -6>, angularTarget=<1 0 0>}
      localJointOffset=<12 0 0>
     }]
    relativeSpaceBone {boneName=pelvis}
   }
   n31 = Anim Dynamics - Bone: dyn_skirt_02_a_tip(ComponentPose=@n30, Alpha=1, ExternalForce="0,0,0") {
    angularSpringConstant=240
    bAngularSpring=true
    bUseGravityOverride=true
    simulationSpace=RootRelative
    alphaBoolBlend {blendOption=Linear}
    boundBone {boneName=dyn_skirt_02_a_tip}
    componentAppliedLinearAccClamp=<3000 3000 3000>
    componentLinearAccScale=<0.1 0.1 0.3>
    componentLinearVelScale=<0.1 0.1 0.1>
    gravityOverride {z=-360}
    physicsBodyDefinitions [{
      boundBone {boneName=dyn_skirt_02_a_tip}
      boxExtents=<24 32 32>
      constraintSetup {angularLimitsMax=<1 2 6>, angularLimitsMin=<-1 -20 -6>, angularTarget=<1 0 0>}
      localJointOffset=<12 0 0>
     }]
    relativeSpaceBone {boneName=pelvis}
   }
   n32 = Anim Dynamics - Bone: dyn_skirt_02_b(ComponentPose=@n31, Alpha=1, ExternalForce="0,0,0") {
    angularSpringConstant=240
    bAngularSpring=true
    bUseGravityOverride=true
    simulationSpace=RootRelative
    alphaBoolBlend {blendOption=Linear}
    boundBone {boneName=dyn_skirt_02_b}
    componentAppliedLinearAccClamp=<3000 3000 3000>
    componentLinearAccScale=<0.1 0.1 0.3>
    componentLinearVelScale=<0.1 0.1 0.1>
    gravityOverride {z=-360}
    physicsBodyDefinitions [{
      boundBone {boneName=dyn_skirt_02_b}
      boxExtents=<24 32 32>
      constraintSetup {angularLimitsMax=<1 2 6>, angularLimitsMin=<-1 -20 -6>, angularTarget=<1 0 0>}
      localJointOffset=<12 0 0>
     }]
    relativeSpaceBone {boneName=pelvis}
   }
   n33 = Anim Dynamics - Bone: dyn_skirt_02_b_tip(ComponentPose=@n32, Alpha=1, ExternalForce="0,0,0") {
    angularSpringConstant=240
    bAngularSpring=true
    bUseGravityOverride=true
    simulationSpace=RootRelative
    alphaBoolBlend {blendOption=Linear}
    boundBone {boneName=dyn_skirt_02_b_tip}
    componentAppliedLinearAccClamp=<3000 3000 3000>
    componentLinearAccScale=<0.1 0.1 0.3>
    componentLinearVelScale=<0.1 0.1 0.1>
    gravityOverride {z=-360}
    physicsBodyDefinitions [{
      boundBone {boneName=dyn_skirt_02_b_tip}
      boxExtents=<24 32 32>
      constraintSetup {angularLimitsMax=<1 2 6>, angularLimitsMin=<-1 -20 -6>, angularTarget=<1 0 0>}
      localJointOffset=<12 0 0>
     }]
    relativeSpaceBone {boneName=pelvis}
   }
   n34 = Anim Dynamics - Bone: dyn_skirt_03(ComponentPose=@n33, Alpha=1, ExternalForce="0,0,0") {
    angularSpringConstant=240
    bAngularSpring=true
    bUseGravityOverride=true
    simulationSpace=RootRelative
    alphaBoolBlend {blendOption=Linear}
    boundBone {boneName=dyn_skirt_03}
    componentAppliedLinearAccClamp=<3000 3000 3000>
    componentLinearAccScale=<0.1 0.1 0.3>
    componentLinearVelScale=<0.1 0.1 0.1>
    gravityOverride {z=-360}
    physicsBodyDefinitions [{
      boundBone {boneName=dyn_skirt_03}
      boxExtents=<24 32 32>
      constraintSetup {angularLimitsMax=<1 2 6>, angularLimitsMin=<-1 -20 -6>, angularTarget=<1 0 0>}
      localJointOffset=<12 0 0>
     }]
    relativeSpaceBone {boneName=pelvis}
   }
   n35 = Anim Dynamics - Bone: dyn_skirt_03_mid(ComponentPose=@n34, Alpha=1, ExternalForce="0,0,0") {
    angularSpringConstant=240
    bAngularSpring=true
    bUseGravityOverride=true
    simulationSpace=RootRelative
    alphaBoolBlend {blendOption=Linear}
    boundBone {boneName=dyn_skirt_03_mid}
    componentAppliedLinearAccClamp=<3000 3000 3000>
    componentLinearAccScale=<0.1 0.1 0.3>
    componentLinearVelScale=<0.1 0.1 0.1>
    gravityOverride {z=-360}
    physicsBodyDefinitions [{
      boundBone {boneName=dyn_skirt_03_mid}
      boxExtents=<24 32 32>
      constraintSetup {angularLimitsMax=<1 2 6>, angularLimitsMin=<-1 -20 -6>, angularTarget=<1 0 0>}
      localJointOffset=<12 0 0>
     }]
    relativeSpaceBone {boneName=pelvis}
   }
   n36 = Anim Dynamics - Bone: dyn_skirt_03_a(ComponentPose=@n35, Alpha=1, ExternalForce="0,0,0") {
    angularSpringConstant=240
    bAngularSpring=true
    bUseGravityOverride=true
    simulationSpace=RootRelative
    alphaBoolBlend {blendOption=Linear}
    boundBone {boneName=dyn_skirt_03_a}
    componentAppliedLinearAccClamp=<3000 3000 3000>
    componentLinearAccScale=<0.1 0.1 0.3>
    componentLinearVelScale=<0.1 0.1 0.1>
    gravityOverride {z=-360}
    physicsBodyDefinitions [{
      boundBone {boneName=dyn_skirt_03_a}
      boxExtents=<24 32 32>
      constraintSetup {angularLimitsMax=<1 2 6>, angularLimitsMin=<-1 -20 -6>, angularTarget=<1 0 0>}
      localJointOffset=<12 0 0>
     }]
    relativeSpaceBone {boneName=pelvis}
   }
   n37 = Anim Dynamics - Bone: dyn_skirt_03_a_tip(ComponentPose=@n36, Alpha=1, ExternalForce="0,0,0") {
    angularSpringConstant=240
    bAngularSpring=true
    bUseGravityOverride=true
    simulationSpace=RootRelative
    alphaBoolBlend {blendOption=Linear}
    boundBone {boneName=dyn_skirt_03_a_tip}
    componentAppliedLinearAccClamp=<3000 3000 3000>
    componentLinearAccScale=<0.1 0.1 0.3>
    componentLinearVelScale=<0.1 0.1 0.1>
    gravityOverride {z=-360}
    physicsBodyDefinitions [{
      boundBone {boneName=dyn_skirt_03_a_tip}
      boxExtents=<24 32 32>
      constraintSetup {angularLimitsMax=<1 2 6>, angularLimitsMin=<-1 -20 -6>, angularTarget=<1 0 0>}
      localJointOffset=<12 0 0>
     }]
    relativeSpaceBone {boneName=pelvis}
   }
   n38 = Anim Dynamics - Bone: dyn_skirt_03_b(ComponentPose=@n37, Alpha=1, ExternalForce="0,0,0") {
    angularSpringConstant=240
    bAngularSpring=true
    bUseGravityOverride=true
    simulationSpace=RootRelative
    alphaBoolBlend {blendOption=Linear}
    boundBone {boneName=dyn_skirt_03_b}
    componentAppliedLinearAccClamp=<3000 3000 3000>
    componentLinearAccScale=<0.1 0.1 0.3>
    componentLinearVelScale=<0.1 0.1 0.1>
    gravityOverride {z=-360}
    physicsBodyDefinitions [{
      boundBone {boneName=dyn_skirt_03_b}
      boxExtents=<24 32 32>
      constraintSetup {angularLimitsMax=<1 2 6>, angularLimitsMin=<-1 -20 -6>, angularTarget=<1 0 0>}
      localJointOffset=<12 0 0>
     }]
    relativeSpaceBone {boneName=pelvis}
   }
   n39 = Anim Dynamics - Bone: dyn_skirt_03_b_tip(ComponentPose=@n38, Alpha=1, ExternalForce="0,0,0") {
    angularSpringConstant=240
    bAngularSpring=true
    bUseGravityOverride=true
    simulationSpace=RootRelative
    alphaBoolBlend {blendOption=Linear}
    boundBone {boneName=dyn_skirt_03_b_tip}
    componentAppliedLinearAccClamp=<3000 3000 3000>
    componentLinearAccScale=<0.1 0.1 0.3>
    componentLinearVelScale=<0.1 0.1 0.1>
    gravityOverride {z=-360}
    physicsBodyDefinitions [{
      boundBone {boneName=dyn_skirt_03_b_tip}
      boxExtents=<24 32 32>
      constraintSetup {angularLimitsMax=<1 2 6>, angularLimitsMin=<-1 -20 -6>, angularTarget=<1 0 0>}
      localJointOffset=<12 0 0>
     }]
    relativeSpaceBone {boneName=pelvis}
   }
   n40 = Anim Dynamics - Bone: dyn_skirt_03_c(ComponentPose=@n39, Alpha=1, ExternalForce="0,0,0") {
    angularSpringConstant=240
    bAngularSpring=true
    bUseGravityOverride=true
    simulationSpace=RootRelative
    alphaBoolBlend {blendOption=Linear}
    boundBone {boneName=dyn_skirt_03_c}
    componentAppliedLinearAccClamp=<3000 3000 3000>
    componentLinearAccScale=<0.1 0.1 0.3>
    componentLinearVelScale=<0.1 0.1 0.1>
    gravityOverride {z=-360}
    physicsBodyDefinitions [{
      boundBone {boneName=dyn_skirt_03_c}
      boxExtents=<24 32 32>
      constraintSetup {angularLimitsMax=<1 2 6>, angularLimitsMin=<-1 -20 -6>, angularTarget=<1 0 0>}
      localJointOffset=<12 0 0>
     }]
    relativeSpaceBone {boneName=pelvis}
   }
   n41 = Anim Dynamics - Bone: dyn_skirt_03_c_tip(ComponentPose=@n40, Alpha=1, ExternalForce="0,0,0") {
    angularSpringConstant=240
    bAngularSpring=true
    bUseGravityOverride=true
    simulationSpace=RootRelative
    alphaBoolBlend {blendOption=Linear}
    boundBone {boneName=dyn_skirt_03_c_tip}
    componentAppliedLinearAccClamp=<3000 3000 3000>
    componentLinearAccScale=<0.1 0.1 0.3>
    componentLinearVelScale=<0.1 0.1 0.1>
    gravityOverride {z=-360}
    physicsBodyDefinitions [{
      boundBone {boneName=dyn_skirt_03_c_tip}
      boxExtents=<24 32 32>
      constraintSetup {angularLimitsMax=<1 2 6>, angularLimitsMin=<-1 -20 -6>, angularTarget=<1 0 0>}
      localJointOffset=<12 0 0>
     }]
    relativeSpaceBone {boneName=pelvis}
   }
   n42 = Anim Dynamics - Bone: dyn_skirt_04(ComponentPose=@n41, Alpha=1, ExternalForce="0,0,0") {
    angularSpringConstant=240
    bAngularSpring=true
    bUseGravityOverride=true
    simulationSpace=RootRelative
    alphaBoolBlend {blendOption=Linear}
    boundBone {boneName=dyn_skirt_04}
    componentAppliedLinearAccClamp=<3000 3000 3000>
    componentLinearAccScale=<0.1 0.1 0.3>
    componentLinearVelScale=<0.1 0.1 0.1>
    gravityOverride {z=-360}
    physicsBodyDefinitions [{
      boundBone {boneName=dyn_skirt_04}
      boxExtents=<24 32 32>
      constraintSetup {angularLimitsMax=<1 2 6>, angularLimitsMin=<-1 -20 -6>, angularTarget=<1 0 0>}
      localJointOffset=<12 0 0>
     }]
    relativeSpaceBone {boneName=pelvis}
   }
   n43 = Anim Dynamics - Bone: dyn_skirt_04_mid(ComponentPose=@n42, Alpha=1, ExternalForce="0,0,0") {
    angularSpringConstant=240
    bAngularSpring=true
    bUseGravityOverride=true
    simulationSpace=RootRelative
    alphaBoolBlend {blendOption=Linear}
    boundBone {boneName=dyn_skirt_04_mid}
    componentAppliedLinearAccClamp=<3000 3000 3000>
    componentLinearAccScale=<0.1 0.1 0.3>
    componentLinearVelScale=<0.1 0.1 0.1>
    gravityOverride {z=-360}
    physicsBodyDefinitions [{
      boundBone {boneName=dyn_skirt_04_mid}
      boxExtents=<24 32 32>
      constraintSetup {angularLimitsMax=<1 2 6>, angularLimitsMin=<-1 -20 -6>, angularTarget=<1 0 0>}
      localJointOffset=<12 0 0>
     }]
    relativeSpaceBone {boneName=pelvis}
   }
   n44 = Anim Dynamics - Bone: dyn_skirt_04_a(ComponentPose=@n43, Alpha=1, ExternalForce="0,0,0") {
    angularSpringConstant=240
    bAngularSpring=true
    bUseGravityOverride=true
    simulationSpace=RootRelative
    alphaBoolBlend {blendOption=Linear}
    boundBone {boneName=dyn_skirt_04_a}
    componentAppliedLinearAccClamp=<3000 3000 3000>
    componentLinearAccScale=<0.1 0.1 0.3>
    componentLinearVelScale=<0.1 0.1 0.1>
    gravityOverride {z=-360}
    physicsBodyDefinitions [{
      boundBone {boneName=dyn_skirt_04_a}
      boxExtents=<24 32 32>
      constraintSetup {angularLimitsMax=<1 2 6>, angularLimitsMin=<-1 -20 -6>, angularTarget=<1 0 0>}
      localJointOffset=<12 0 0>
     }]
    relativeSpaceBone {boneName=pelvis}
   }
   n45 = Anim Dynamics - Bone: dyn_skirt_04_a_tip(ComponentPose=@n44, Alpha=1, ExternalForce="0,0,0") {
    angularSpringConstant=240
    bAngularSpring=true
    bUseGravityOverride=true
    simulationSpace=RootRelative
    alphaBoolBlend {blendOption=Linear}
    boundBone {boneName=dyn_skirt_04_a_tip}
    componentAppliedLinearAccClamp=<3000 3000 3000>
    componentLinearAccScale=<0.1 0.1 0.3>
    componentLinearVelScale=<0.1 0.1 0.1>
    gravityOverride {z=-360}
    physicsBodyDefinitions [{
      boundBone {boneName=dyn_skirt_04_a_tip}
      boxExtents=<24 32 32>
      constraintSetup {angularLimitsMax=<1 2 6>, angularLimitsMin=<-1 -20 -6>, angularTarget=<1 0 0>}
      localJointOffset=<12 0 0>
     }]
    relativeSpaceBone {boneName=pelvis}
   }
   n46 = Anim Dynamics - Bone: dyn_skirt_04_b(ComponentPose=@n45, Alpha=1, ExternalForce="0,0,0") {
    angularSpringConstant=240
    bAngularSpring=true
    bUseGravityOverride=true
    simulationSpace=RootRelative
    alphaBoolBlend {blendOption=Linear}
    boundBone {boneName=dyn_skirt_04_b}
    componentAppliedLinearAccClamp=<3000 3000 3000>
    componentLinearAccScale=<0.1 0.1 0.3>
    componentLinearVelScale=<0.1 0.1 0.1>
    gravityOverride {z=-360}
    physicsBodyDefinitions [{
      boundBone {boneName=dyn_skirt_04_b}
      boxExtents=<24 32 32>
      constraintSetup {angularLimitsMax=<1 2 6>, angularLimitsMin=<-1 -20 -6>, angularTarget=<1 0 0>}
      localJointOffset=<12 0 0>
     }]
    relativeSpaceBone {boneName=pelvis}
   }
   n47 = Anim Dynamics - Bone: dyn_skirt_04_b_tip(ComponentPose=@n46, Alpha=1, ExternalForce="0,0,0") {
    angularSpringConstant=240
    bAngularSpring=true
    bUseGravityOverride=true
    simulationSpace=RootRelative
    alphaBoolBlend {blendOption=Linear}
    boundBone {boneName=dyn_skirt_04_b_tip}
    componentAppliedLinearAccClamp=<3000 3000 3000>
    componentLinearAccScale=<0.1 0.1 0.3>
    componentLinearVelScale=<0.1 0.1 0.1>
    gravityOverride {z=-360}
    physicsBodyDefinitions [{
      boundBone {boneName=dyn_skirt_04_b_tip}
      boxExtents=<24 32 32>
      constraintSetup {angularLimitsMax=<1 2 6>, angularLimitsMin=<-1 -20 -6>, angularTarget=<1 0 0>}
      localJointOffset=<12 0 0>
     }]
    relativeSpaceBone {boneName=pelvis}
   }
   n48 = Anim Dynamics - Bone: dyn_skirt_04_c(ComponentPose=@n47, Alpha=1, ExternalForce="0,0,0") {
    angularSpringConstant=240
    bAngularSpring=true
    bUseGravityOverride=true
    simulationSpace=RootRelative
    alphaBoolBlend {blendOption=Linear}
    boundBone {boneName=dyn_skirt_04_c}
    componentAppliedLinearAccClamp=<3000 3000 3000>
    componentLinearAccScale=<0.1 0.1 0.3>
    componentLinearVelScale=<0.1 0.1 0.1>
    gravityOverride {z=-360}
    physicsBodyDefinitions [{
      boundBone {boneName=dyn_skirt_04_c}
      boxExtents=<24 32 32>
      constraintSetup {angularLimitsMax=<1 2 6>, angularLimitsMin=<-1 -20 -6>, angularTarget=<1 0 0>}
      localJointOffset=<12 0 0>
     }]
    relativeSpaceBone {boneName=pelvis}
   }
   n49 = Anim Dynamics - Bone: dyn_skirt_04_c_tip(ComponentPose=@n48, Alpha=1, ExternalForce="0,0,0") {
    angularSpringConstant=240
    bAngularSpring=true
    bUseGravityOverride=true
    simulationSpace=RootRelative
    alphaBoolBlend {blendOption=Linear}
    boundBone {boneName=dyn_skirt_04_c_tip}
    componentAppliedLinearAccClamp=<3000 3000 3000>
    componentLinearAccScale=<0.1 0.1 0.3>
    componentLinearVelScale=<0.1 0.1 0.1>
    gravityOverride {z=-360}
    physicsBodyDefinitions [{
      boundBone {boneName=dyn_skirt_04_c_tip}
      boxExtents=<24 32 32>
      constraintSetup {angularLimitsMax=<1 2 6>, angularLimitsMin=<-1 -20 -6>, angularTarget=<1 0 0>}
      localJointOffset=<12 0 0>
     }]
    relativeSpaceBone {boneName=pelvis}
   }
   n50 = Anim Dynamics - Bone: dyn_skirt_04_d(ComponentPose=@n49, Alpha=1, ExternalForce="0,0,0") {
    angularSpringConstant=240
    bAngularSpring=true
    bUseGravityOverride=true
    simulationSpace=RootRelative
    alphaBoolBlend {blendOption=Linear}
    boundBone {boneName=dyn_skirt_04_d}
    componentAppliedLinearAccClamp=<3000 3000 3000>
    componentLinearAccScale=<0.1 0.1 0.3>
    componentLinearVelScale=<0.1 0.1 0.1>
    gravityOverride {z=-360}
    physicsBodyDefinitions [{
      boundBone {boneName=dyn_skirt_04_d}
      boxExtents=<24 32 32>
      constraintSetup {angularLimitsMax=<1 2 6>, angularLimitsMin=<-1 -20 -6>, angularTarget=<1 0 0>}
      localJointOffset=<12 0 0>
     }]
    relativeSpaceBone {boneName=pelvis}
   }
   n51 = Anim Dynamics - Bone: dyn_skirt_04_d_tip(ComponentPose=@n50, Alpha=1, ExternalForce="0,0,0") {
    angularSpringConstant=240
    bAngularSpring=true
    bUseGravityOverride=true
    simulationSpace=RootRelative
    alphaBoolBlend {blendOption=Linear}
    boundBone {boneName=dyn_skirt_04_d_tip}
    componentAppliedLinearAccClamp=<3000 3000 3000>
    componentLinearAccScale=<0.1 0.1 0.3>
    componentLinearVelScale=<0.1 0.1 0.1>
    gravityOverride {z=-360}
    physicsBodyDefinitions [{
      boundBone {boneName=dyn_skirt_04_d_tip}
      boxExtents=<24 32 32>
      constraintSetup {angularLimitsMax=<1 2 6>, angularLimitsMin=<-1 -20 -6>, angularTarget=<1 0 0>}
      localJointOffset=<12 0 0>
     }]
    relativeSpaceBone {boneName=pelvis}
   }
   n52 = Anim Dynamics - Bone: dyn_skirt_05(ComponentPose=@n51, Alpha=1, ExternalForce="0,0,0") {
    angularSpringConstant=240
    bAngularSpring=true
    bUseGravityOverride=true
    simulationSpace=RootRelative
    alphaBoolBlend {blendOption=Linear}
    boundBone {boneName=dyn_skirt_05}
    componentAppliedLinearAccClamp=<3000 3000 3000>
    componentLinearAccScale=<0.1 0.1 0.3>
    componentLinearVelScale=<0.1 0.1 0.1>
    gravityOverride {z=-360}
    physicsBodyDefinitions [{
      boundBone {boneName=dyn_skirt_05}
      boxExtents=<24 32 32>
      constraintSetup {angularLimitsMax=<1 2 6>, angularLimitsMin=<-1 -20 -6>, angularTarget=<1 0 0>}
      localJointOffset=<12 0 0>
     }]
    relativeSpaceBone {boneName=pelvis}
   }
   n53 = Anim Dynamics - Bone: dyn_skirt_05_mid(ComponentPose=@n52, Alpha=1, ExternalForce="0,0,0") {
    angularSpringConstant=240
    bAngularSpring=true
    bUseGravityOverride=true
    simulationSpace=RootRelative
    alphaBoolBlend {blendOption=Linear}
    boundBone {boneName=dyn_skirt_05_mid}
    componentAppliedLinearAccClamp=<3000 3000 3000>
    componentLinearAccScale=<0.1 0.1 0.3>
    componentLinearVelScale=<0.1 0.1 0.1>
    gravityOverride {z=-360}
    physicsBodyDefinitions [{
      boundBone {boneName=dyn_skirt_05_mid}
      boxExtents=<24 32 32>
      constraintSetup {angularLimitsMax=<1 2 6>, angularLimitsMin=<-1 -20 -6>, angularTarget=<1 0 0>}
      localJointOffset=<12 0 0>
     }]
    relativeSpaceBone {boneName=pelvis}
   }
   n54 = Anim Dynamics - Bone: dyn_skirt_05_a(ComponentPose=@n53, Alpha=1, ExternalForce="0,0,0") {
    angularSpringConstant=240
    bAngularSpring=true
    bUseGravityOverride=true
    simulationSpace=RootRelative
    alphaBoolBlend {blendOption=Linear}
    boundBone {boneName=dyn_skirt_05_a}
    componentAppliedLinearAccClamp=<3000 3000 3000>
    componentLinearAccScale=<0.1 0.1 0.3>
    componentLinearVelScale=<0.1 0.1 0.1>
    gravityOverride {z=-360}
    physicsBodyDefinitions [{
      boundBone {boneName=dyn_skirt_05_a}
      boxExtents=<24 32 32>
      constraintSetup {angularLimitsMax=<1 2 6>, angularLimitsMin=<-1 -20 -6>, angularTarget=<1 0 0>}
      localJointOffset=<12 0 0>
     }]
    relativeSpaceBone {boneName=pelvis}
   }
   n55 = Anim Dynamics - Bone: dyn_skirt_05_a_tip(ComponentPose=@n54, Alpha=1, ExternalForce="0,0,0") {
    angularSpringConstant=240
    bAngularSpring=true
    bUseGravityOverride=true
    simulationSpace=RootRelative
    alphaBoolBlend {blendOption=Linear}
    boundBone {boneName=dyn_skirt_05_a_tip}
    componentAppliedLinearAccClamp=<3000 3000 3000>
    componentLinearAccScale=<0.1 0.1 0.3>
    componentLinearVelScale=<0.1 0.1 0.1>
    gravityOverride {z=-360}
    physicsBodyDefinitions [{
      boundBone {boneName=dyn_skirt_05_a_tip}
      boxExtents=<24 32 32>
      constraintSetup {angularLimitsMax=<1 2 6>, angularLimitsMin=<-1 -20 -6>, angularTarget=<1 0 0>}
      localJointOffset=<12 0 0>
     }]
    relativeSpaceBone {boneName=pelvis}
   }
   n56 = Anim Dynamics - Bone: dyn_skirt_05_b(ComponentPose=@n55, Alpha=1, ExternalForce="0,0,0") {
    angularSpringConstant=240
    bAngularSpring=true
    bUseGravityOverride=true
    simulationSpace=RootRelative
    alphaBoolBlend {blendOption=Linear}
    boundBone {boneName=dyn_skirt_05_b}
    componentAppliedLinearAccClamp=<3000 3000 3000>
    componentLinearAccScale=<0.1 0.1 0.3>
    componentLinearVelScale=<0.1 0.1 0.1>
    gravityOverride {z=-360}
    physicsBodyDefinitions [{
      boundBone {boneName=dyn_skirt_05_b}
      boxExtents=<24 32 32>
      constraintSetup {angularLimitsMax=<1 2 6>, angularLimitsMin=<-1 -20 -6>, angularTarget=<1 0 0>}
      localJointOffset=<12 0 0>
     }]
    relativeSpaceBone {boneName=pelvis}
   }
   n57 = Anim Dynamics - Bone: dyn_skirt_05_b_tip(ComponentPose=@n56, Alpha=1, ExternalForce="0,0,0") {
    angularSpringConstant=240
    bAngularSpring=true
    bUseGravityOverride=true
    simulationSpace=RootRelative
    alphaBoolBlend {blendOption=Linear}
    boundBone {boneName=dyn_skirt_05_b_tip}
    componentAppliedLinearAccClamp=<3000 3000 3000>
    componentLinearAccScale=<0.1 0.1 0.3>
    componentLinearVelScale=<0.1 0.1 0.1>
    gravityOverride {z=-360}
    physicsBodyDefinitions [{
      boundBone {boneName=dyn_skirt_05_b_tip}
      boxExtents=<24 32 32>
      constraintSetup {angularLimitsMax=<1 2 6>, angularLimitsMin=<-1 -20 -6>, angularTarget=<1 0 0>}
      localJointOffset=<12 0 0>
     }]
    relativeSpaceBone {boneName=pelvis}
   }
   n58 = Anim Dynamics - Bone: dyn_skirt_05_c(ComponentPose=@n57, Alpha=1, ExternalForce="0,0,0") {
    angularSpringConstant=240
    bAngularSpring=true
    bUseGravityOverride=true
    simulationSpace=RootRelative
    alphaBoolBlend {blendOption=Linear}
    boundBone {boneName=dyn_skirt_05_c}
    componentAppliedLinearAccClamp=<3000 3000 3000>
    componentLinearAccScale=<0.1 0.1 0.3>
    componentLinearVelScale=<0.1 0.1 0.1>
    gravityOverride {z=-360}
    physicsBodyDefinitions [{
      boundBone {boneName=dyn_skirt_05_c}
      boxExtents=<24 32 32>
      constraintSetup {angularLimitsMax=<1 2 6>, angularLimitsMin=<-1 -20 -6>, angularTarget=<1 0 0>}
      localJointOffset=<12 0 0>
     }]
    relativeSpaceBone {boneName=pelvis}
   }
   n59 = Anim Dynamics - Bone: dyn_skirt_05_c_tip(ComponentPose=@n58, Alpha=1, ExternalForce="0,0,0") {
    angularSpringConstant=240
    bAngularSpring=true
    bUseGravityOverride=true
    simulationSpace=RootRelative
    alphaBoolBlend {blendOption=Linear}
    boundBone {boneName=dyn_skirt_05_c_tip}
    componentAppliedLinearAccClamp=<3000 3000 3000>
    componentLinearAccScale=<0.1 0.1 0.3>
    componentLinearVelScale=<0.1 0.1 0.1>
    gravityOverride {z=-360}
    physicsBodyDefinitions [{
      boundBone {boneName=dyn_skirt_05_c_tip}
      boxExtents=<24 32 32>
      constraintSetup {angularLimitsMax=<1 2 6>, angularLimitsMin=<-1 -20 -6>, angularTarget=<1 0 0>}
      localJointOffset=<12 0 0>
     }]
    relativeSpaceBone {boneName=pelvis}
   }
   n60 = Anim Dynamics - Bone: dyn_skirt_06(ComponentPose=@n59, Alpha=1, ExternalForce="0,0,0") {
    angularSpringConstant=240
    bAngularSpring=true
    bUseGravityOverride=true
    simulationSpace=RootRelative
    alphaBoolBlend {blendOption=Linear}
    boundBone {boneName=dyn_skirt_06}
    componentAppliedLinearAccClamp=<3000 3000 3000>
    componentLinearAccScale=<0.1 0.1 0.3>
    componentLinearVelScale=<0.1 0.1 0.1>
    gravityOverride {z=-360}
    physicsBodyDefinitions [{
      boundBone {boneName=dyn_skirt_06}
      boxExtents=<24 32 32>
      constraintSetup {angularLimitsMax=<1 2 6>, angularLimitsMin=<-1 -20 -6>, angularTarget=<1 0 0>}
      localJointOffset=<12 0 0>
     }]
    relativeSpaceBone {boneName=pelvis}
   }
   n61 = Anim Dynamics - Bone: dyn_skirt_06_mid(ComponentPose=@n60, Alpha=1, ExternalForce="0,0,0") {
    angularSpringConstant=240
    bAngularSpring=true
    bUseGravityOverride=true
    simulationSpace=RootRelative
    alphaBoolBlend {blendOption=Linear}
    boundBone {boneName=dyn_skirt_06_mid}
    componentAppliedLinearAccClamp=<3000 3000 3000>
    componentLinearAccScale=<0.1 0.1 0.3>
    componentLinearVelScale=<0.1 0.1 0.1>
    gravityOverride {z=-360}
    physicsBodyDefinitions [{
      boundBone {boneName=dyn_skirt_06_mid}
      boxExtents=<24 32 32>
      constraintSetup {angularLimitsMax=<1 2 6>, angularLimitsMin=<-1 -20 -6>, angularTarget=<1 0 0>}
      localJointOffset=<12 0 0>
     }]
    relativeSpaceBone {boneName=pelvis}
   }
   n62 = Anim Dynamics - Bone: dyn_skirt_06_a(ComponentPose=@n61, Alpha=1, ExternalForce="0,0,0") {
    angularSpringConstant=240
    bAngularSpring=true
    bUseGravityOverride=true
    simulationSpace=RootRelative
    alphaBoolBlend {blendOption=Linear}
    boundBone {boneName=dyn_skirt_06_a}
    componentAppliedLinearAccClamp=<3000 3000 3000>
    componentLinearAccScale=<0.1 0.1 0.3>
    componentLinearVelScale=<0.1 0.1 0.1>
    gravityOverride {z=-360}
    physicsBodyDefinitions [{
      boundBone {boneName=dyn_skirt_06_a}
      boxExtents=<24 32 32>
      constraintSetup {angularLimitsMax=<1 2 6>, angularLimitsMin=<-1 -20 -6>, angularTarget=<1 0 0>}
      localJointOffset=<12 0 0>
     }]
    relativeSpaceBone {boneName=pelvis}
   }
   n63 = Anim Dynamics - Bone: dyn_skirt_06_a_tip(ComponentPose=@n62, Alpha=1, ExternalForce="0,0,0") {
    angularSpringConstant=240
    bAngularSpring=true
    bUseGravityOverride=true
    simulationSpace=RootRelative
    alphaBoolBlend {blendOption=Linear}
    boundBone {boneName=dyn_skirt_06_a_tip}
    componentAppliedLinearAccClamp=<3000 3000 3000>
    componentLinearAccScale=<0.1 0.1 0.3>
    componentLinearVelScale=<0.1 0.1 0.1>
    gravityOverride {z=-360}
    physicsBodyDefinitions [{
      boundBone {boneName=dyn_skirt_06_a_tip}
      boxExtents=<24 32 32>
      constraintSetup {angularLimitsMax=<1 2 6>, angularLimitsMin=<-1 -20 -6>, angularTarget=<1 0 0>}
      localJointOffset=<12 0 0>
     }]
    relativeSpaceBone {boneName=pelvis}
   }
   n64 = Anim Dynamics - Bone: dyn_skirt_06_b(ComponentPose=@n63, Alpha=1, ExternalForce="0,0,0") {
    angularSpringConstant=240
    bAngularSpring=true
    bUseGravityOverride=true
    simulationSpace=RootRelative
    alphaBoolBlend {blendOption=Linear}
    boundBone {boneName=dyn_skirt_06_b}
    componentAppliedLinearAccClamp=<3000 3000 3000>
    componentLinearAccScale=<0.1 0.1 0.3>
    componentLinearVelScale=<0.1 0.1 0.1>
    gravityOverride {z=-360}
    physicsBodyDefinitions [{
      boundBone {boneName=dyn_skirt_06_b}
      boxExtents=<24 32 32>
      constraintSetup {angularLimitsMax=<1 2 6>, angularLimitsMin=<-1 -20 -6>, angularTarget=<1 0 0>}
      localJointOffset=<12 0 0>
     }]
    relativeSpaceBone {boneName=pelvis}
   }
   n65 = Anim Dynamics - Bone: dyn_skirt_06_b_tip(ComponentPose=@n64, Alpha=1, ExternalForce="0,0,0") {
    angularSpringConstant=240
    bAngularSpring=true
    bUseGravityOverride=true
    simulationSpace=RootRelative
    alphaBoolBlend {blendOption=Linear}
    boundBone {boneName=dyn_skirt_06_b_tip}
    componentAppliedLinearAccClamp=<3000 3000 3000>
    componentLinearAccScale=<0.1 0.1 0.3>
    componentLinearVelScale=<0.1 0.1 0.1>
    gravityOverride {z=-360}
    physicsBodyDefinitions [{
      boundBone {boneName=dyn_skirt_06_b_tip}
      boxExtents=<24 32 32>
      constraintSetup {angularLimitsMax=<1 2 6>, angularLimitsMin=<-1 -20 -6>, angularTarget=<1 0 0>}
      localJointOffset=<12 0 0>
     }]
    relativeSpaceBone {boneName=pelvis}
   }
   n66 = Anim Dynamics - Bone: dyn_skirt_07(ComponentPose=@n65, Alpha=1, ExternalForce="0,0,0") {
    angularSpringConstant=240
    bAngularSpring=true
    bUseGravityOverride=true
    simulationSpace=RootRelative
    alphaBoolBlend {blendOption=Linear}
    boundBone {boneName=dyn_skirt_07}
    componentAppliedLinearAccClamp=<3000 3000 3000>
    componentLinearAccScale=<0.1 0.1 0.3>
    componentLinearVelScale=<0.1 0.1 0.1>
    gravityOverride {z=-360}
    physicsBodyDefinitions [{
      boundBone {boneName=dyn_skirt_07}
      boxExtents=<24 32 32>
      constraintSetup {angularLimitsMax=<1 2 6>, angularLimitsMin=<-1 -20 -6>, angularTarget=<1 0 0>}
      localJointOffset=<12 0 0>
     }]
    relativeSpaceBone {boneName=pelvis}
   }
   n67 = Anim Dynamics - Bone: dyn_skirt_07_mid(ComponentPose=@n66, Alpha=1, ExternalForce="0,0,0") {
    angularSpringConstant=240
    bAngularSpring=true
    bUseGravityOverride=true
    simulationSpace=RootRelative
    alphaBoolBlend {blendOption=Linear}
    boundBone {boneName=dyn_skirt_07_mid}
    componentAppliedLinearAccClamp=<3000 3000 3000>
    componentLinearAccScale=<0.1 0.1 0.3>
    componentLinearVelScale=<0.1 0.1 0.1>
    gravityOverride {z=-360}
    physicsBodyDefinitions [{
      boundBone {boneName=dyn_skirt_07_mid}
      boxExtents=<24 32 32>
      constraintSetup {angularLimitsMax=<1 2 6>, angularLimitsMin=<-1 -20 -6>, angularTarget=<1 0 0>}
      localJointOffset=<12 0 0>
     }]
    relativeSpaceBone {boneName=pelvis}
   }
   n68 = Anim Dynamics - Bone: dyn_skirt_07_a(ComponentPose=@n67, Alpha=1, ExternalForce="0,0,0") {
    angularSpringConstant=240
    bAngularSpring=true
    bUseGravityOverride=true
    simulationSpace=RootRelative
    alphaBoolBlend {blendOption=Linear}
    boundBone {boneName=dyn_skirt_07_a}
    componentAppliedLinearAccClamp=<3000 3000 3000>
    componentLinearAccScale=<0.1 0.1 0.3>
    componentLinearVelScale=<0.1 0.1 0.1>
    gravityOverride {z=-360}
    physicsBodyDefinitions [{
      boundBone {boneName=dyn_skirt_07_a}
      boxExtents=<24 32 32>
      constraintSetup {angularLimitsMax=<1 2 6>, angularLimitsMin=<-1 -20 -6>, angularTarget=<1 0 0>}
      localJointOffset=<12 0 0>
     }]
    relativeSpaceBone {boneName=pelvis}
   }
   n69 = Anim Dynamics - Bone: dyn_skirt_07_a_tip(ComponentPose=@n68, Alpha=1, ExternalForce="0,0,0") {
    angularSpringConstant=240
    bAngularSpring=true
    bUseGravityOverride=true
    simulationSpace=RootRelative
    alphaBoolBlend {blendOption=Linear}
    boundBone {boneName=dyn_skirt_07_a_tip}
    componentAppliedLinearAccClamp=<3000 3000 3000>
    componentLinearAccScale=<0.1 0.1 0.3>
    componentLinearVelScale=<0.1 0.1 0.1>
    gravityOverride {z=-360}
    physicsBodyDefinitions [{
      boundBone {boneName=dyn_skirt_07_a_tip}
      boxExtents=<24 32 32>
      constraintSetup {angularLimitsMax=<1 2 6>, angularLimitsMin=<-1 -20 -6>, angularTarget=<1 0 0>}
      localJointOffset=<12 0 0>
     }]
    relativeSpaceBone {boneName=pelvis}
   }
   n70 = Anim Dynamics - Bone: dyn_skirt_07_b(ComponentPose=@n69, Alpha=1, ExternalForce="0,0,0") {
    angularSpringConstant=240
    bAngularSpring=true
    bUseGravityOverride=true
    simulationSpace=RootRelative
    alphaBoolBlend {blendOption=Linear}
    boundBone {boneName=dyn_skirt_07_b}
    componentAppliedLinearAccClamp=<3000 3000 3000>
    componentLinearAccScale=<0.1 0.1 0.3>
    componentLinearVelScale=<0.1 0.1 0.1>
    gravityOverride {z=-360}
    physicsBodyDefinitions [{
      boundBone {boneName=dyn_skirt_07_b}
      boxExtents=<24 32 32>
      constraintSetup {angularLimitsMax=<1 2 6>, angularLimitsMin=<-1 -20 -6>, angularTarget=<1 0 0>}
      localJointOffset=<12 0 0>
     }]
    relativeSpaceBone {boneName=pelvis}
   }
   n71 = Anim Dynamics - Bone: dyn_skirt_07_b_tip(ComponentPose=@n70, Alpha=1, ExternalForce="0,0,0") {
    angularSpringConstant=240
    bAngularSpring=true
    bUseGravityOverride=true
    simulationSpace=RootRelative
    alphaBoolBlend {blendOption=Linear}
    boundBone {boneName=dyn_skirt_07_b_tip}
    componentAppliedLinearAccClamp=<3000 3000 3000>
    componentLinearAccScale=<0.1 0.1 0.3>
    componentLinearVelScale=<0.1 0.1 0.1>
    gravityOverride {z=-360}
    physicsBodyDefinitions [{
      boundBone {boneName=dyn_skirt_07_b_tip}
      boxExtents=<24 32 32>
      constraintSetup {angularLimitsMax=<1 2 6>, angularLimitsMin=<-1 -20 -6>, angularTarget=<1 0 0>}
      localJointOffset=<12 0 0>
     }]
    relativeSpaceBone {boneName=pelvis}
   }
   n72 = Anim Dynamics - Bone: dyn_skirt_07_c(ComponentPose=@n71, Alpha=1, ExternalForce="0,0,0") {
    angularSpringConstant=240
    bAngularSpring=true
    bUseGravityOverride=true
    simulationSpace=RootRelative
    alphaBoolBlend {blendOption=Linear}
    boundBone {boneName=dyn_skirt_07_c}
    componentAppliedLinearAccClamp=<3000 3000 3000>
    componentLinearAccScale=<0.1 0.1 0.3>
    componentLinearVelScale=<0.1 0.1 0.1>
    gravityOverride {z=-360}
    physicsBodyDefinitions [{
      boundBone {boneName=dyn_skirt_07_c}
      boxExtents=<24 32 32>
      constraintSetup {angularLimitsMax=<1 2 6>, angularLimitsMin=<-1 -20 -6>, angularTarget=<1 0 0>}
      localJointOffset=<12 0 0>
     }]
    relativeSpaceBone {boneName=pelvis}
   }
   n73 = Anim Dynamics - Bone: dyn_skirt_07_c_tip(ComponentPose=@n72, Alpha=1, ExternalForce="0,0,0") {
    angularSpringConstant=240
    bAngularSpring=true
    bUseGravityOverride=true
    simulationSpace=RootRelative
    alphaBoolBlend {blendOption=Linear}
    boundBone {boneName=dyn_skirt_07_c_tip}
    componentAppliedLinearAccClamp=<3000 3000 3000>
    componentLinearAccScale=<0.1 0.1 0.3>
    componentLinearVelScale=<0.1 0.1 0.1>
    gravityOverride {z=-360}
    physicsBodyDefinitions [{
      boundBone {boneName=dyn_skirt_07_c_tip}
      boxExtents=<24 32 32>
      constraintSetup {angularLimitsMax=<1 2 6>, angularLimitsMin=<-1 -20 -6>, angularTarget=<1 0 0>}
      localJointOffset=<12 0 0>
     }]
    relativeSpaceBone {boneName=pelvis}
   }
   n74 = Anim Dynamics - Bone: dyn_r_collar_01_a(ComponentPose=@n73, Alpha=1, ExternalForce="0,0,0") {
    bUseGravityOverride=true
    simulationSpace=RootRelative
    alphaBoolBlend {blendOption=Linear}
    boundBone {boneName=dyn_r_collar_01_a}
    componentLinearAccScale=<1 1 1>
    componentLinearVelScale=<1 1 1>
    gravityOverride {z=-60}
    physicsBodyDefinitions [{
      boundBone {boneName=dyn_r_collar_01_a}
      boxExtents=<16 20 20>
      constraintSetup {angularLimitsMax=<2 5 5>, angularLimitsMin=<-2 -5 -5>, angularTarget=<0 0 0>}
      localJointOffset=<9 0 0>
     }]
    relativeSpaceBone {boneName=pelvis}
   }
   n75 = Anim Dynamics - Bone: dyn_r_collar_01_b(ComponentPose=@n74, Alpha=1, ExternalForce="0,0,0") {
    bUseGravityOverride=true
    simulationSpace=RootRelative
    alphaBoolBlend {blendOption=Linear}
    boundBone {boneName=dyn_r_collar_01_b}
    componentLinearAccScale=<1 1 1>
    componentLinearVelScale=<1 1 1>
    gravityOverride {z=-60}
    physicsBodyDefinitions [{
      boundBone {boneName=dyn_r_collar_01_b}
      boxExtents=<16 20 20>
      constraintSetup {angularLimitsMax=<2 5 5>, angularLimitsMin=<-2 -5 -5>, angularTarget=<0 0 0>}
      localJointOffset=<9 0 0>
     }]
    relativeSpaceBone {boneName=pelvis}
   }
   n76 = Anim Dynamics - Bone: dyn_r_collar_01_c(ComponentPose=@n75, Alpha=1, ExternalForce="0,0,0") {
    bUseGravityOverride=true
    simulationSpace=RootRelative
    alphaBoolBlend {blendOption=Linear}
    boundBone {boneName=dyn_r_collar_01_c}
    componentLinearAccScale=<1 1 1>
    componentLinearVelScale=<1 1 1>
    gravityOverride {z=-60}
    physicsBodyDefinitions [{
      boundBone {boneName=dyn_r_collar_01_c}
      boxExtents=<16 20 20>
      constraintSetup {angularLimitsMax=<2 5 5>, angularLimitsMin=<-2 -5 -5>, angularTarget=<0 0 0>}
      localJointOffset=<9 0 0>
     }]
    relativeSpaceBone {boneName=pelvis}
   }
   n77 = Anim Dynamics - Bone: dyn_r_collar_01_d(ComponentPose=@n76, Alpha=1, ExternalForce="0,0,0") {
    bUseGravityOverride=true
    simulationSpace=RootRelative
    alphaBoolBlend {blendOption=Linear}
    boundBone {boneName=dyn_r_collar_01_d}
    componentLinearAccScale=<1 1 1>
    componentLinearVelScale=<1 1 1>
    gravityOverride {z=-60}
    physicsBodyDefinitions [{
      boundBone {boneName=dyn_r_collar_01_d}
      boxExtents=<16 20 20>
      constraintSetup {angularLimitsMax=<2 5 5>, angularLimitsMin=<-2 -5 -5>, angularTarget=<0 0 0>}
      localJointOffset=<9 0 0>
     }]
    relativeSpaceBone {boneName=pelvis}
   }
   n78 = Anim Dynamics - Bone: dyn_r_collar_01_e(ComponentPose=@n77, Alpha=1, ExternalForce="0,0,0") {
    bUseGravityOverride=true
    simulationSpace=RootRelative
    alphaBoolBlend {blendOption=Linear}
    boundBone {boneName=dyn_r_collar_01_e}
    componentLinearAccScale=<1 1 1>
    componentLinearVelScale=<1 1 1>
    gravityOverride {z=-60}
    physicsBodyDefinitions [{
      boundBone {boneName=dyn_r_collar_01_e}
      boxExtents=<16 20 20>
      constraintSetup {angularLimitsMax=<2 5 5>, angularLimitsMin=<-2 -5 -5>, angularTarget=<0 0 0>}
      localJointOffset=<9 0 0>
     }]
    relativeSpaceBone {boneName=pelvis}
   }
   n79 = Anim Dynamics - Bone: dyn_r_collar_01_f(ComponentPose=@n78, Alpha=1, ExternalForce="0,0,0") {
    bUseGravityOverride=true
    simulationSpace=RootRelative
    alphaBoolBlend {blendOption=Linear}
    boundBone {boneName=dyn_r_collar_01_f}
    componentLinearAccScale=<1 1 1>
    componentLinearVelScale=<1 1 1>
    gravityOverride {z=-60}
    physicsBodyDefinitions [{
      boundBone {boneName=dyn_r_collar_01_f}
      boxExtents=<16 20 20>
      constraintSetup {angularLimitsMax=<2 5 5>, angularLimitsMin=<-2 -5 -5>, angularTarget=<0 0 0>}
      localJointOffset=<9 0 0>
     }]
    relativeSpaceBone {boneName=pelvis}
   }
   n80 = Anim Dynamics - Bone: dyn_r_collar_02_a(ComponentPose=@n79, Alpha=1, ExternalForce="0,0,0") {
    bUseGravityOverride=true
    simulationSpace=RootRelative
    alphaBoolBlend {blendOption=Linear}
    boundBone {boneName=dyn_r_collar_02_a}
    componentLinearAccScale=<1 1 1>
    componentLinearVelScale=<1 1 1>
    gravityOverride {z=-60}
    physicsBodyDefinitions [{
      boundBone {boneName=dyn_r_collar_02_a}
      boxExtents=<16 20 20>
      constraintSetup {angularLimitsMax=<2 5 5>, angularLimitsMin=<-2 -5 -5>, angularTarget=<0 0 0>}
      localJointOffset=<9 0 0>
     }]
    relativeSpaceBone {boneName=pelvis}
   }
   n81 = Anim Dynamics - Bone: dyn_r_collar_02_b(ComponentPose=@n80, Alpha=1, ExternalForce="0,0,0") {
    bUseGravityOverride=true
    simulationSpace=RootRelative
    alphaBoolBlend {blendOption=Linear}
    boundBone {boneName=dyn_r_collar_02_b}
    componentLinearAccScale=<1 1 1>
    componentLinearVelScale=<1 1 1>
    gravityOverride {z=-60}
    physicsBodyDefinitions [{
      boundBone {boneName=dyn_r_collar_02_b}
      boxExtents=<16 20 20>
      constraintSetup {angularLimitsMax=<2 5 5>, angularLimitsMin=<-2 -5 -5>, angularTarget=<0 0 0>}
      localJointOffset=<9 0 0>
     }]
    relativeSpaceBone {boneName=pelvis}
   }
   n82 = Anim Dynamics - Bone: dyn_r_collar_02_c(ComponentPose=@n81, Alpha=1, ExternalForce="0,0,0") {
    bUseGravityOverride=true
    simulationSpace=RootRelative
    alphaBoolBlend {blendOption=Linear}
    boundBone {boneName=dyn_r_collar_02_c}
    componentLinearAccScale=<1 1 1>
    componentLinearVelScale=<1 1 1>
    gravityOverride {z=-60}
    physicsBodyDefinitions [{
      boundBone {boneName=dyn_r_collar_02_c}
      boxExtents=<16 20 20>
      constraintSetup {angularLimitsMax=<2 5 5>, angularLimitsMin=<-2 -5 -5>, angularTarget=<0 0 0>}
      localJointOffset=<9 0 0>
     }]
    relativeSpaceBone {boneName=pelvis}
   }
   n83 = Anim Dynamics - Bone: dyn_r_collar_02_d(ComponentPose=@n82, Alpha=1, ExternalForce="0,0,0") {
    bUseGravityOverride=true
    simulationSpace=RootRelative
    alphaBoolBlend {blendOption=Linear}
    boundBone {boneName=dyn_r_collar_02_d}
    componentLinearAccScale=<1 1 1>
    componentLinearVelScale=<1 1 1>
    gravityOverride {z=-60}
    physicsBodyDefinitions [{
      boundBone {boneName=dyn_r_collar_02_d}
      boxExtents=<16 20 20>
      constraintSetup {angularLimitsMax=<2 5 5>, angularLimitsMin=<-2 -5 -5>, angularTarget=<0 0 0>}
      localJointOffset=<9 0 0>
     }]
    relativeSpaceBone {boneName=pelvis}
   }
   n84 = Anim Dynamics - Bone: dyn_r_collar_02_e(ComponentPose=@n83, Alpha=1, ExternalForce="0,0,0") {
    bUseGravityOverride=true
    simulationSpace=RootRelative
    alphaBoolBlend {blendOption=Linear}
    boundBone {boneName=dyn_r_collar_02_e}
    componentLinearAccScale=<1 1 1>
    componentLinearVelScale=<1 1 1>
    gravityOverride {z=-60}
    physicsBodyDefinitions [{
      boundBone {boneName=dyn_r_collar_02_e}
      boxExtents=<16 20 20>
      constraintSetup {angularLimitsMax=<2 5 5>, angularLimitsMin=<-2 -5 -5>, angularTarget=<0 0 0>}
      localJointOffset=<9 0 0>
     }]
    relativeSpaceBone {boneName=pelvis}
   }
   n85 = Anim Dynamics - Bone: dyn_r_collar_02_f(ComponentPose=@n84, Alpha=1, ExternalForce="0,0,0") {
    bUseGravityOverride=true
    simulationSpace=RootRelative
    alphaBoolBlend {blendOption=Linear}
    boundBone {boneName=dyn_r_collar_02_f}
    componentLinearAccScale=<1 1 1>
    componentLinearVelScale=<1 1 1>
    gravityOverride {z=-60}
    physicsBodyDefinitions [{
      boundBone {boneName=dyn_r_collar_02_f}
      boxExtents=<16 20 20>
      constraintSetup {angularLimitsMax=<2 5 5>, angularLimitsMin=<-2 -5 -5>, angularTarget=<0 0 0>}
      localJointOffset=<9 0 0>
     }]
    relativeSpaceBone {boneName=pelvis}
   }
   n86 = Anim Dynamics - Bone: dyn_r_collar_03_a(ComponentPose=@n85, Alpha=1, ExternalForce="0,0,0") {
    bUseGravityOverride=true
    simulationSpace=RootRelative
    alphaBoolBlend {blendOption=Linear}
    boundBone {boneName=dyn_r_collar_03_a}
    componentLinearAccScale=<1 1 1>
    componentLinearVelScale=<1 1 1>
    gravityOverride {z=-60}
    physicsBodyDefinitions [{
      boundBone {boneName=dyn_r_collar_03_a}
      boxExtents=<16 20 20>
      constraintSetup {angularLimitsMax=<2 5 5>, angularLimitsMin=<-2 -5 -5>, angularTarget=<0 0 0>}
      localJointOffset=<9 0 0>
     }]
    relativeSpaceBone {boneName=pelvis}
   }
   n87 = Anim Dynamics - Bone: dyn_r_collar_03_b(ComponentPose=@n86, Alpha=1, ExternalForce="0,0,0") {
    bUseGravityOverride=true
    simulationSpace=RootRelative
    alphaBoolBlend {blendOption=Linear}
    boundBone {boneName=dyn_r_collar_03_b}
    componentLinearAccScale=<1 1 1>
    componentLinearVelScale=<1 1 1>
    gravityOverride {z=-60}
    physicsBodyDefinitions [{
      boundBone {boneName=dyn_r_collar_03_b}
      boxExtents=<16 20 20>
      constraintSetup {angularLimitsMax=<2 5 5>, angularLimitsMin=<-2 -5 -5>, angularTarget=<0 0 0>}
      localJointOffset=<9 0 0>
     }]
    relativeSpaceBone {boneName=pelvis}
   }
   n88 = Anim Dynamics - Bone: dyn_r_collar_03_c(ComponentPose=@n87, Alpha=1, ExternalForce="0,0,0") {
    bUseGravityOverride=true
    simulationSpace=RootRelative
    alphaBoolBlend {blendOption=Linear}
    boundBone {boneName=dyn_r_collar_03_c}
    componentLinearAccScale=<1 1 1>
    componentLinearVelScale=<1 1 1>
    gravityOverride {z=-60}
    physicsBodyDefinitions [{
      boundBone {boneName=dyn_r_collar_03_c}
      boxExtents=<16 20 20>
      constraintSetup {angularLimitsMax=<2 5 5>, angularLimitsMin=<-2 -5 -5>, angularTarget=<0 0 0>}
      localJointOffset=<9 0 0>
     }]
    relativeSpaceBone {boneName=pelvis}
   }
   n89 = Anim Dynamics - Bone: dyn_r_collar_03_d(ComponentPose=@n88, Alpha=1, ExternalForce="0,0,0") {
    bUseGravityOverride=true
    simulationSpace=RootRelative
    alphaBoolBlend {blendOption=Linear}
    boundBone {boneName=dyn_r_collar_03_d}
    componentLinearAccScale=<1 1 1>
    componentLinearVelScale=<1 1 1>
    gravityOverride {z=-60}
    physicsBodyDefinitions [{
      boundBone {boneName=dyn_r_collar_03_d}
      boxExtents=<16 20 20>
      constraintSetup {angularLimitsMax=<2 5 5>, angularLimitsMin=<-2 -5 -5>, angularTarget=<0 0 0>}
      localJointOffset=<9 0 0>
     }]
    relativeSpaceBone {boneName=pelvis}
   }
   n90 = Anim Dynamics - Bone: dyn_r_collar_03_e(ComponentPose=@n89, Alpha=1, ExternalForce="0,0,0") {
    bUseGravityOverride=true
    simulationSpace=RootRelative
    alphaBoolBlend {blendOption=Linear}
    boundBone {boneName=dyn_r_collar_03_e}
    componentLinearAccScale=<1 1 1>
    componentLinearVelScale=<1 1 1>
    gravityOverride {z=-60}
    physicsBodyDefinitions [{
      boundBone {boneName=dyn_r_collar_03_e}
      boxExtents=<16 20 20>
      constraintSetup {angularLimitsMax=<2 5 5>, angularLimitsMin=<-2 -5 -5>, angularTarget=<0 0 0>}
      localJointOffset=<9 0 0>
     }]
    relativeSpaceBone {boneName=pelvis}
   }
   n91 = Anim Dynamics - Bone: dyn_r_collar_03_f(ComponentPose=@n90, Alpha=1, ExternalForce="0,0,0") {
    bUseGravityOverride=true
    simulationSpace=RootRelative
    alphaBoolBlend {blendOption=Linear}
    boundBone {boneName=dyn_r_collar_03_f}
    componentLinearAccScale=<1 1 1>
    componentLinearVelScale=<1 1 1>
    gravityOverride {z=-60}
    physicsBodyDefinitions [{
      boundBone {boneName=dyn_r_collar_03_f}
      boxExtents=<16 20 20>
      constraintSetup {angularLimitsMax=<2 5 5>, angularLimitsMin=<-2 -5 -5>, angularTarget=<0 0 0>}
      localJointOffset=<9 0 0>
     }]
    relativeSpaceBone {boneName=pelvis}
   }
   n92 = Anim Dynamics - Bone: dyn_r_collar_03_g(ComponentPose=@n91, Alpha=1, ExternalForce="0,0,0") {
    bUseGravityOverride=true
    simulationSpace=RootRelative
    alphaBoolBlend {blendOption=Linear}
    boundBone {boneName=dyn_r_collar_03_g}
    componentLinearAccScale=<1 1 1>
    componentLinearVelScale=<1 1 1>
    gravityOverride {z=-60}
    physicsBodyDefinitions [{
      boundBone {boneName=dyn_r_collar_03_g}
      boxExtents=<16 20 20>
      constraintSetup {angularLimitsMax=<2 5 5>, angularLimitsMin=<-2 -5 -5>, angularTarget=<0 0 0>}
      localJointOffset=<9 0 0>
     }]
    relativeSpaceBone {boneName=pelvis}
   }
   n93 = Anim Dynamics - Bone: dyn_l_collar_01_a(ComponentPose=@n92, Alpha=1, ExternalForce="0,0,0") {
    bUseGravityOverride=true
    simulationSpace=RootRelative
    alphaBoolBlend {blendOption=Linear}
    boundBone {boneName=dyn_l_collar_01_a}
    componentLinearAccScale=<1 1 1>
    componentLinearVelScale=<1 1 1>
    gravityOverride {z=-60}
    physicsBodyDefinitions [{
      boundBone {boneName=dyn_l_collar_01_a}
      boxExtents=<16 20 20>
      constraintSetup {angularLimitsMax=<2 5 5>, angularLimitsMin=<-2 -5 -5>, angularTarget=<0 0 0>}
      localJointOffset=<9 0 0>
     }]
    relativeSpaceBone {boneName=pelvis}
   }
   n94 = Anim Dynamics - Bone: dyn_l_collar_01_b(ComponentPose=@n93, Alpha=1, ExternalForce="0,0,0") {
    bUseGravityOverride=true
    simulationSpace=RootRelative
    alphaBoolBlend {blendOption=Linear}
    boundBone {boneName=dyn_l_collar_01_b}
    componentLinearAccScale=<1 1 1>
    componentLinearVelScale=<1 1 1>
    gravityOverride {z=-60}
    physicsBodyDefinitions [{
      boundBone {boneName=dyn_l_collar_01_b}
      boxExtents=<16 20 20>
      constraintSetup {angularLimitsMax=<2 5 5>, angularLimitsMin=<-2 -5 -5>, angularTarget=<0 0 0>}
      localJointOffset=<9 0 0>
     }]
    relativeSpaceBone {boneName=pelvis}
   }
   n95 = Anim Dynamics - Bone: dyn_l_collar_02_a(ComponentPose=@n94, Alpha=1, ExternalForce="0,0,0") {
    bUseGravityOverride=true
    simulationSpace=RootRelative
    alphaBoolBlend {blendOption=Linear}
    boundBone {boneName=dyn_l_collar_02_a}
    componentLinearAccScale=<1 1 1>
    componentLinearVelScale=<1 1 1>
    gravityOverride {z=-60}
    physicsBodyDefinitions [{
      boundBone {boneName=dyn_l_collar_02_a}
      boxExtents=<16 20 20>
      constraintSetup {angularLimitsMax=<2 5 5>, angularLimitsMin=<-2 -5 -5>, angularTarget=<0 0 0>}
      localJointOffset=<9 0 0>
     }]
    relativeSpaceBone {boneName=pelvis}
   }
   n96 = Anim Dynamics - Bone: dyn_l_collar_02_b(ComponentPose=@n95, Alpha=1, ExternalForce="0,0,0") {
    bUseGravityOverride=true
    simulationSpace=RootRelative
    alphaBoolBlend {blendOption=Linear}
    boundBone {boneName=dyn_l_collar_02_b}
    componentLinearAccScale=<1 1 1>
    componentLinearVelScale=<1 1 1>
    gravityOverride {z=-60}
    physicsBodyDefinitions [{
      boundBone {boneName=dyn_l_collar_02_b}
      boxExtents=<16 20 20>
      constraintSetup {angularLimitsMax=<2 5 5>, angularLimitsMin=<-2 -5 -5>, angularTarget=<0 0 0>}
      localJointOffset=<9 0 0>
     }]
    relativeSpaceBone {boneName=pelvis}
   }
   n97 = Anim Dynamics - Bone: dyn_l_collar_03_a(ComponentPose=@n96, Alpha=1, ExternalForce="0,0,0") {
    bUseGravityOverride=true
    simulationSpace=RootRelative
    alphaBoolBlend {blendOption=Linear}
    boundBone {boneName=dyn_l_collar_03_a}
    componentLinearAccScale=<1 1 1>
    componentLinearVelScale=<1 1 1>
    gravityOverride {z=-60}
    physicsBodyDefinitions [{
      boundBone {boneName=dyn_l_collar_03_a}
      boxExtents=<16 20 20>
      constraintSetup {angularLimitsMax=<2 5 5>, angularLimitsMin=<-2 -5 -5>, angularTarget=<0 0 0>}
      localJointOffset=<9 0 0>
     }]
    relativeSpaceBone {boneName=pelvis}
   }
   n98 = Anim Dynamics - Bone: dyn_l_collar_03_b(ComponentPose=@n97, Alpha=1, ExternalForce="0,0,0") {
    bUseGravityOverride=true
    simulationSpace=RootRelative
    alphaBoolBlend {blendOption=Linear}
    boundBone {boneName=dyn_l_collar_03_b}
    componentLinearAccScale=<1 1 1>
    componentLinearVelScale=<1 1 1>
    gravityOverride {z=-60}
    physicsBodyDefinitions [{
      boundBone {boneName=dyn_l_collar_03_b}
      boxExtents=<16 20 20>
      constraintSetup {angularLimitsMax=<2 5 5>, angularLimitsMin=<-2 -5 -5>, angularTarget=<0 0 0>}
      localJointOffset=<9 0 0>
     }]
    relativeSpaceBone {boneName=pelvis}
   }
   Output Pose(Result=Component To Local(ComponentPose=@n98)) {layerGroup=None}
   n99 = Control Rig(Source=Copy Pose From Mesh() {bUseAttachedParent=true}) {
    alphaBoolBlend {blendOption=Linear}
    controlRigAssetReference {blueprintRigClass=&/F_MED_DeformRig_Virtuous_ControlRig.F_MED_DeformRig_Virtuous_ControlRig_C}
   }
   Save cached pose 'SavedPose12245'(Pose=@n99) {cacheName=SavedPose12245}
  ]}
 EventGraph {nodes [
   Event BlueprintInitializeAnimation()
   > Set SkelMesh(SkelMesh=GetOwnerSkeletalMeshComponent())
   n1 = TryGetPawnOwner()
   > Is Valid(InputObject=@n1)
   > n2 = Cast To FortPlayerPawnAthena(Object=@n1)
   > Set PlayerPawn(PlayerPawn:Fort Player Pawn Athena Object Reference=@n2) -> (Output_Get:Fort Player Pawn Athena Object Reference)
   n3 = Event BlueprintUpdateAnimation()
   > Is Valid(InputObject=Get PlayerPawn() -> (PlayerPawn:Fort Player Pawn Athena Object Reference))
   > Set WPO Globals Function(
    Material=Get MatInstanceDynamic() -> (MatInstanceDynamic:Material Instance Dynamic Object Reference)
    Overlay Material=Get OverlayMID() -> (OverlayMID:Material Instance Dynamic Object Reference)
    Skeletal Mesh=Get SkelMesh()
    "Local Noise Transform  Location"=Get Local Noise Transform Location() -> (Local Noise Transform Location:Vector)
    Local Noise Offset=Get Local Noise Offset()
   )
   > Set WPO Regional Function!(
    execute
    Material:Material Instance Dynamic Object Reference=Get MatInstanceDynamic()
    Overlay Material:Material Instance Dynamic Object Reference=Get OverlayMID()
    Skeletal Mesh:Skeletal Mesh Component Object Reference=Get SkelMesh()
    Velocity Transform:Name=foot_r
    Velocity Interp Speed:double=Get Velocity Interp Speed()
    Collision Transform:Name=calf_r
    Collision Transform Axis:EAxis Enum
    Invert Collision Transform Axis:bool=true
    Delta Time X:double=@n3.DeltaTimeX
    Previous Velocity Transform Location:"Vector (by ref)"=Get R1 Velocity Transform Location()
    Previous Velocity Transform Velocity:"Vector (by ref)"=Get R1 Velocity Transform Velocity()
    Region:int=1
   ) -> (then)
   > n4 = Branch(Condition=Get bIsSkyDiving())
   > n5 = Set WPO Pho Wind Intensity(WPO Pho Wind Intensity=1)
   SetScalarParameterValue(execute=[@n6.then, @n5.then], self=Get MatInstanceDynamic(), ParameterName=ClothSim_WindStrength, Value=Get WPO Pho Wind Intensity())
   > Set Crouching(Crouching:double="To Float (Boolean)"(InBool=Get bIsCrouching())) -> (Output_Get:double)
   n6 = Set WPO Pho Wind Intensity(execute=@n4.else, WPO Pho Wind Intensity=0)
   Event OnCharacterCustomizationLoaded()
   > n7 = Set SkelMesh(SkelMesh=GetPartSkeletalMeshComponent())
   > n8 = Get Dynamic Material Instance(SkeletalMeshComponent=@n7)
   > Set MatInstanceDynamic(MatInstanceDynamic:Material Instance Dynamic Object Reference=@n8) -> (Output_Get:Material Instance Dynamic Object Reference)
   > n9 = GetOrCreateDynamicOverlayMaterial(MeshComponent=@n7, MaterialName=None)
   > Set OverlayMID(OverlayMID:Material Instance Dynamic Object Reference=@n9) -> (Output_Get:Material Instance Dynamic Object Reference)
  ]}
]
variables [
 Crouching {type=number}
 Local Noise Offset {type=Vector}
 Local Noise Transform Location {type=Vector}
 MatInstanceDynamic {type=MaterialInstanceDynamic}
 OverlayMID {type=MaterialInstanceDynamic}
 PlayerPawn {type=FortPlayerPawnAthena}
 R1 Velocity Transform Location {type=Vector}
 R1 Velocity Transform Velocity {type=Vector}
 R2 Velocity Transform Location {type=Vector}
 R2 Velocity Transform Velocity {type=Vector}
 SkelMesh {type=SkeletalMeshComponent}
 Velocity Interp Speed {type=number, default=30}
 WPO Pho Wind Intensity {type=number}
]
