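# Control Rig for the F_MED_Virtuous face accessory that applies corrective offsets to hair bones as the head moves.
# Two spherical pose readers track the head bone, one relative to neck_01, the other to clavicle_r, outputting falloff weights.
# The neck reader weights Modify Transforms on back and side hair bones; the clavicle reader weights right bang and side-hair master bones.
# The clavicle reader has debug drawing left enabled.
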ControlRigBlueprint /BRCosmetics/Characters/Player/Female/Medium/Bodies/F_MED_Virtuous/Meshes/Parts/F_MED_Virtuous_FaceAcc_CtrlRig
graphs [
 RigVMFunctionLibrary {links [], nodes []}
 RigVMModel {nodes [
   RigUnit_BeginExecution = &RigUnit_BeginExecution()
   n1 = &RigUnit_SphericalPoseReader(
    ActiveRegionScaleFactors=(PositiveWidth=1,NegativeWidth=1,PositiveHeight=1,NegativeHeight=1)
    ActiveRegionSize=0.1
    Debug=(bDrawDebug=false,bDraw2D=false,bDrawLocalAxes=false,DebugScale=25,DebugSegments=20,DebugThickness=0.25)
    DriverAxis=(X=1,Y=0,Z=0)
    DriverItem=(Type=Bone,Name="head")
    FalloffRegionScaleFactors=(PositiveWidth=1,NegativeWidth=1,PositiveHeight=1,NegativeHeight=1)
    FalloffSize=0.3
    FlipHeightScaling=true
    FlipWidthScaling=true
    OptionalParentItem=(Type=Bone,Name="neck_01")
    RotationOffset=(X=0,Y=-90,Z=-45)
    ExecutePin=@RigUnit_BeginExecution.ExecutePin
   )
   > n2 = &RigUnit_SphericalPoseReader(
    ActiveRegionScaleFactors=(PositiveWidth=1,NegativeWidth=1,PositiveHeight=1,NegativeHeight=1)
    ActiveRegionSize=0.05
    Debug=(bDrawDebug=True,bDraw2D=false,bDrawLocalAxes=false,DebugScale=25,DebugSegments=20,DebugThickness=0.25)
    DriverAxis=(X=0,Y=0,Z=1)
    DriverItem=(Type=Bone,Name="head")
    FalloffRegionScaleFactors=(PositiveWidth=1,NegativeWidth=1,PositiveHeight=1,NegativeHeight=1)
    FalloffSize=0.15
    FlipHeightScaling=true
    FlipWidthScaling=true
    OptionalParentItem=(Type=Bone,Name="clavicle_r")
    RotationOffset=(X=-60,Y=0,Z=-60)
   )
   > &RigUnit_ModifyTransforms(
    ItemToModify=((Item=(Type=Bone,Name="dyn_hair_back_01"),Transform=(Rotation=(X=0,Y=0,Z=0,W=1),Translation=(X=-2,Y=0,Z=4.5),Scale3D=(X=1,Y=1,Z=1))))
    ItemToModify.0.Transform.Rotation=&RigVMFunction_MathQuaternionFromEuler(Euler=(X=0,Y=-30,Z=0)).Result
    Weight=@n1.OutputParam
   )
   > &RigUnit_ModifyTransforms(
    ItemToModify=((Item=(Type=Bone,Name="dyn_hair_back_02"),Transform=(Rotation=(X=0,Y=0,Z=0,W=1),Translation=(X=0,Y=0,Z=4),Scale3D=(X=1,Y=1,Z=1))))
    ItemToModify.0.Transform.Rotation=&RigVMFunction_MathQuaternionFromEuler(Euler=(X=0,Y=-20,Z=0)).Result
    Weight=@n1.OutputParam
   )
   > &RigUnit_ModifyTransforms(
    ItemToModify=((Item=(Type=Bone,Name="dyn_hair_r_01"),Transform=(Rotation=(X=0,Y=0,Z=0,W=1),Translation=(X=-3,Y=0,Z=0),Scale3D=(X=0.75,Y=1,Z=1))))
    ItemToModify.0.Transform.Rotation=&RigVMFunction_MathQuaternionFromEuler(Euler=(X=0,Y=0,Z=30)).Result
    Weight=@n1.OutputParam
   )
   > &RigUnit_ModifyTransforms(
    ItemToModify=((Item=(Type=Bone,Name="dyn_hair_l_01"),Transform=(Rotation=(X=0,Y=0,Z=0,W=1),Translation=(X=-3,Y=0,Z=0),Scale3D=(X=0.75,Y=1,Z=1))))
    ItemToModify.0.Transform.Rotation=&RigVMFunction_MathQuaternionFromEuler(Euler=(X=0,Y=0,Z=-30)).Result
    Weight=@n1.OutputParam
   )
   > &RigUnit_ModifyTransforms(
    ItemToModify=((Item=(Type=Bone,Name="dyn_bang_r_MSTR"),Transform=(Rotation=(X=0,Y=0,Z=0,W=1),Translation=(X=0,Y=0,Z=3),Scale3D=(X=1,Y=1,Z=1))))
    ItemToModify.0.Transform.Rotation=&RigVMFunction_MathQuaternionFromEuler(Euler=(X=0,Y=-30,Z=0)).Result
    Weight=@n2.OutputParam
   )
   > &RigUnit_ModifyTransforms(
    ItemToModify=((Item=(Type=Bone,Name="dyn_hair_r_MSTR"),Transform=(Rotation=(X=0,Y=0,Z=0,W=1),Translation=(X=-2,Y=0,Z=-4),Scale3D=(X=0.5,Y=1,Z=1))))
    ItemToModify.0.Transform.Rotation=&RigVMFunction_MathQuaternionFromEuler(Euler=(X=0,Y=-30,Z=0)).Result
    Weight=@n2.OutputParam
   )
  ]}
]
hierarchy [
 root {loc=<0 0 0>, rot=<0 0 0 1>}
 pelvis {parent=root, loc=<0 -0.5301 98.0571>, rot=<0.0127 0.707 -0.0127 -0.707>}
 spine_01 {parent=pelvis, loc=<6.1208 -0.9328 0>, rot=<0 0 0.034 -0.9994>}
 spine_02 {parent=spine_01, loc=<6.5996 -0.2095 0>, rot=<0 0 0.0238 -0.9997>}
 spine_03 {parent=spine_02, loc=<6.4914 -0.1237 0>, rot=<0 0 0.0149 -0.9999>}
 spine_04 {parent=spine_03, loc=<6.9598 0.0794 0>, rot=<0 0 -0.0132 -0.9999>}
 spine_05 {parent=spine_04, loc=<6.9987 0 0>, rot=<0 0 -0.0015 -1>}
 clavicle_l {parent=spine_05, loc=<5.4831 -0.1053 -3.2946>, rot=<0.0465 -0.7289 -0.009 -0.6829>}
 upperarm_l {parent=clavicle_l, loc=<9.8658 -1.2136 -0.4105>, rot=<0.0164 -0.3245 -0.0269 -0.9454>}
 lowerarm_l {parent=upperarm_l, loc=<26.465 -0.8345 0.259>, rot=<0.0036 0.0161 -0.2298 -0.9731>}
 hand_l {parent=lowerarm_l, loc=<22.7355 -0.3203 -0.6105>, rot=<-0.6784 -0.0223 0.0735 -0.7307>}
 thumb_01_l {parent=hand_l, loc=<1.3118 -0.3191 -2.459>, rot=<0.6546 -0.4042 0.0664 -0.6354>}
 thumb_02_l {parent=thumb_01_l, loc=<5.9336 -0.0448 0.0401>, rot=<-0.0037 0.0213 0.298 -0.9543>}
 thumb_03_l {parent=thumb_02_l, loc=<2.2223 0.3682 -0.0907>, rot=<0.0118 -0.013 0.0066 -0.9998>}
 index_metacarpal_l {parent=hand_l, loc=<1.6535 0.0183 -2.96>, rot=<0.0158 -0.0124 -0.0173 -0.9996>}
 index_01_l {parent=index_metacarpal_l, loc=<8.3008 0.0357 -0.0524>, rot=<-0.0044 0.0485 0.0924 -0.9945>}
 index_02_l {parent=index_01_l, loc=<3.3587 -0.3004 0.0452>, rot=<-0.002 0.0027 0.3225 -0.9466>}
 index_03_l {parent=index_02_l, loc=<2.5578 -0.123 -0.0252>, rot=<-0.002 0.0067 0.2558 -0.9667>}
 middle_metacarpal_l {parent=hand_l, loc=<2.1856 -0.1725 -0.5984>, rot=<0.006 0.0168 -0.0329 -0.9993>}
 middle_01_l {parent=middle_metacarpal_l, loc=<7.3275 -0.2447 0.0129>, rot=<-0.0208 0.0042 0.0858 -0.9961>}
 middle_02_l {parent=middle_01_l, loc=<4.0123 0.1401 0.1175>, rot=<-0.0218 0.0258 0.3512 -0.9357>}
 middle_03_l {parent=middle_02_l, loc=<2.6077 0.0534 0.1373>, rot=<0.0082 0.0188 0.1708 -0.9851>}
 ring_metacarpal_l {parent=hand_l, loc=<2.2074 -0.3625 1.3638>, rot=<-0.0284 0.0722 -0.0458 -0.9959>}
 ring_01_l {parent=ring_metacarpal_l, loc=<7.0295 -0.6197 -0.1446>, rot=<-0.0489 -0.0252 0.1892 -0.9804>}
 ring_02_l {parent=ring_01_l, loc=<3.6174 0.4647 -0.1027>, rot=<0 0 0.2806 -0.9598>}
 ring_03_l {parent=ring_02_l, loc=<2.4152 0.114 0.0199>, rot=<0 0 0.1189 -0.9929>}
 pinky_metacarpal_l {parent=hand_l, loc=<1.7616 -0.9753 3.3814>, rot=<-0.0397 0.0814 -0.069 -0.9935>}
 pinky_01_l {parent=pinky_metacarpal_l, loc=<7.1709 -0.4831 -0.1094>, rot=<-0.0654 -0.0035 0.2242 -0.9723>}
 pinky_02_l {parent=pinky_01_l, loc=<2.8274 0.0893 -0.1043>, rot=<0.0037 0.0011 0.2854 -0.9584>}
 pinky_03_l {parent=pinky_02_l, loc=<2.2046 0.0307 -0.0078>, rot=<-0.0002 0.0098 0.1445 -0.9895>}
 lowerarm_twist_02_l {parent=lowerarm_l, loc=<10.6707 0 0>, rot=<0 0 0 1>}
 lowerarm_twist_01_l {parent=lowerarm_l, loc=<18.316 0 0>, rot=<0 0 0 1>}
 upperarm_twist_01_l {parent=upperarm_l, loc=<7.2135 -0.1134 0>, rot=<0 0 0 1>}
 upperarm_twist_02_l {parent=upperarm_l, loc=<19.7998 -0.1134 0>, rot=<0 0 0 1>}
 neck_01 {parent=spine_05, loc=<11.3138 0.1529 0>, rot=<0 0 -0.2003 -0.9797>}
 neck_02 {parent=neck_01, loc=<4.5627 0 0>, rot=<0 0 0.0504 -0.9987>}
 head {parent=neck_02, loc=<5.3914 0 0>, rot=<0 0 0.115 -0.9934>}
 dyn_hair_01_MSTR {parent=head, loc=<-0.1584 -9.7712 8.1025>, rot=<0.1891 -0.9055 -0.3734 0.0699>}
 dyn_hair_01 {parent=dyn_hair_01_MSTR, loc=<0 0 0>, rot=<0 0 0 1>}
 dyn_hair_02_MSTR {parent=head, loc=<-4.5132 -11.1396 3.6499>, rot=<0.1456 -0.7065 -0.6926 0>}
 dyn_hair_02 {parent=dyn_hair_02_MSTR, loc=<0 0 0>, rot=<0 0 0 1>}
 dyn_hair_03_MSTR {parent=head, loc=<-6.5303 -8.9088 3.9671>, rot=<0.0629 -0.7909 -0.6068 0.0483>}
 dyn_hair_03_a {parent=dyn_hair_03_MSTR, loc=<0 0 0>, rot=<0 0 0 1>}
 dyn_hair_03_b {parent=dyn_hair_03_a, loc=<2.9618 0 0>, rot=<-0.0201 0.0277 0.1249 -0.9916>}
 dyn_hair_04_MSTR {parent=head, loc=<-6.5303 -8.9088 -2.9693>, rot=<-0.032 0.6049 0.7914 -0.0817>}
 dyn_hair_04_a {parent=dyn_hair_04_MSTR, loc=<0 0 0>, rot=<0 0 0 1>}
 dyn_hair_04_b {parent=dyn_hair_04_a, loc=<2.7774 0 0>, rot=<0.0049 0.093 -0.0378 -0.9949>}
 dyn_hair_04_c {parent=dyn_hair_04_b, loc=<3.1007 0 0>, rot=<0.0075 -0.0161 -0.0667 -0.9976>}
 dyn_hair_05_MSTR {parent=head, loc=<-1.5001 -12.0165 -3.9708>, rot=<-0.068 0.5005 0.863 -0.0124>}
 dyn_hair_05 {parent=dyn_hair_05_MSTR, loc=<0 0 0>, rot=<0 0 0 1>}
 dyn_hair_06_MSTR {parent=head, loc=<-2.4482 -7.6654 -6.6868>, rot=<-0.196 0.3918 0.8742 -0.2092>}
 dyn_hair_06 {parent=dyn_hair_06_MSTR, loc=<0 0 0>, rot=<0 0 0 1>}
 dyn_hair_07_MSTR {parent=head, loc=<4.8613 -12.0853 -4.1787>, rot=<-0.1873 0.5558 0.8097 -0.0208>}
 dyn_hair_07 {parent=dyn_hair_07_MSTR, loc=<0 0 0>, rot=<0 0 0 1>}
 dyn_bang_r_MSTR {parent=head, loc=<-2.0384 -1.7722 6.7688>, rot=<0.0041 -0.9899 -0.0006 -0.1416>}
 dyn_bang_r_01 {parent=dyn_bang_r_MSTR, loc=<0 0 0>, rot=<0 0 0 1>}
 dyn_bang_r_02 {parent=dyn_bang_r_01, loc=<3.4577 0 0>, rot=<-0.0068 -0.31 -0.0209 -0.9505>}
 dyn_bang_l_MSTR {parent=head, loc=<-2.0386 -1.7723 -6.7688>, rot=<0.1416 0.0006 0.9899 -0.0041>}
 dyn_bang_l_01 {parent=dyn_bang_l_MSTR, loc=<0 0 0>, rot=<0 0 0 1>}
 dyn_bang_l_02 {parent=dyn_bang_l_01, loc=<3.4577 0 0>, rot=<-0.0477 -0.2894 -0.1556 -0.9433>}
 dyn_hair_back_MSTR {parent=head, loc=<1.3521 -9.5105 0>, rot=<0.0029 -0.7071 0.7071 -0.0029>}
 dyn_hair_back_01 {parent=dyn_hair_back_MSTR, loc=<0 0 0>, rot=<0 0 0 1>}
 dyn_hair_back_02 {parent=dyn_hair_back_01, loc=<5.0905 0 0>, rot=<0 0 0 1>}
 dyn_hair_back_03 {parent=dyn_hair_back_02, loc=<5.2025 0 0>, rot=<0 0 0 1>}
 dyn_hair_r_MSTR {parent=head, loc=<1.3265 -6.4117 7.6427>, rot=<0 0 1 -0.0041>}
 dyn_hair_r_01 {parent=dyn_hair_r_MSTR, loc=<0 0 0>, rot=<0 0 0 1>}
 dyn_hair_r_02 {parent=dyn_hair_r_01, loc=<5.0905 0 0>, rot=<0 0 0 1>}
 dyn_hair_l_MSTR {parent=head, loc=<1.3263 -6.4117 -7.643>, rot=<0.0041 -1 0 0>}
 dyn_hair_l_01 {parent=dyn_hair_l_MSTR, loc=<0 0 0>, rot=<0 0 0 1>}
 dyn_hair_l_02 {parent=dyn_hair_l_01, loc=<5.0905 0 0>, rot=<0 0 0 1>}
 clavicle_r {parent=spine_05, loc=<5.4831 -0.1053 3.2946>, rot=<0.7289 0.0465 -0.6829 0.009>}
 upperarm_r {parent=clavicle_r, loc=<-9.8658 1.2136 0.4105>, rot=<0.0164 -0.3245 -0.0269 -0.9454>}
 lowerarm_r {parent=upperarm_r, loc=<-26.465 0.8345 -0.259>, rot=<0.0036 0.0161 -0.2298 -0.9731>}
 hand_r {parent=lowerarm_r, loc=<-22.7355 0.3203 0.6106>, rot=<-0.6784 -0.0223 0.0735 -0.7307>}
 thumb_01_r {parent=hand_r, loc=<-1.3118 0.3191 2.459>, rot=<0.6546 -0.4042 0.0664 -0.6354>}
 thumb_02_r {parent=thumb_01_r, loc=<-5.9336 0.0448 -0.0401>, rot=<-0.0037 0.0213 0.298 -0.9543>}
 thumb_03_r {parent=thumb_02_r, loc=<-2.2223 -0.3682 0.0907>, rot=<0.0118 -0.013 0.0066 -0.9998>}
 index_metacarpal_r {parent=hand_r, loc=<-1.6535 -0.0183 2.96>, rot=<0.0158 -0.0124 -0.0173 -0.9996>}
 index_01_r {parent=index_metacarpal_r, loc=<-8.3008 -0.0357 0.0524>, rot=<-0.0044 0.0485 0.0924 -0.9945>}
 index_02_r {parent=index_01_r, loc=<-3.3587 0.3004 -0.0452>, rot=<-0.002 0.0027 0.3225 -0.9466>}
 index_03_r {parent=index_02_r, loc=<-2.5578 0.123 0.0252>, rot=<-0.002 0.0067 0.2558 -0.9667>}
 middle_metacarpal_r {parent=hand_r, loc=<-2.1856 0.1725 0.5984>, rot=<0.006 0.0168 -0.0329 -0.9993>}
 middle_01_r {parent=middle_metacarpal_r, loc=<-7.3275 0.2447 -0.0129>, rot=<-0.0208 0.0042 0.0858 -0.9961>}
 middle_02_r {parent=middle_01_r, loc=<-4.0123 -0.1401 -0.1175>, rot=<-0.0218 0.0258 0.3512 -0.9357>}
 middle_03_r {parent=middle_02_r, loc=<-2.6077 -0.0534 -0.1373>, rot=<0.0082 0.0188 0.1708 -0.9851>}
 ring_metacarpal_r {parent=hand_r, loc=<-2.2074 0.3625 -1.3638>, rot=<-0.0284 0.0722 -0.0458 -0.9959>}
 ring_01_r {parent=ring_metacarpal_r, loc=<-7.0295 0.6197 0.1446>, rot=<-0.0489 -0.0252 0.1892 -0.9804>}
 ring_02_r {parent=ring_01_r, loc=<-3.6174 -0.4647 0.1027>, rot=<0 0 0.2806 -0.9598>}
 ring_03_r {parent=ring_02_r, loc=<-2.4152 -0.114 -0.0199>, rot=<0 0 0.1189 -0.9929>}
 pinky_metacarpal_r {parent=hand_r, loc=<-1.7616 0.9753 -3.3814>, rot=<-0.0397 0.0814 -0.069 -0.9935>}
 pinky_01_r {parent=pinky_metacarpal_r, loc=<-7.1709 0.4831 0.1094>, rot=<-0.0654 -0.0035 0.2242 -0.9723>}
 pinky_02_r {parent=pinky_01_r, loc=<-2.8274 -0.0893 0.1043>, rot=<0.0037 0.0011 0.2854 -0.9584>}
 pinky_03_r {parent=pinky_02_r, loc=<-2.2046 -0.0307 0.0078>, rot=<-0.0002 0.0098 0.1445 -0.9895>}
 lowerarm_twist_02_r {parent=lowerarm_r, loc=<-10.6707 0 0>, rot=<0 0 0 -1>}
 lowerarm_twist_01_r {parent=lowerarm_r, loc=<-18.316 0 0>, rot=<0 0 0 -1>}
 upperarm_twist_01_r {parent=upperarm_r, loc=<-7.2135 0.1134 0>, rot=<0 0 0 -1>}
 upperarm_twist_02_r {parent=upperarm_r, loc=<-19.7998 0.1134 0>, rot=<0 0 0 -1>}
 thigh_l {parent=pelvis, loc=<-1.0855 -0.2109 -8.7053>, rot=<0.0009 0.0355 0.0155 -0.9993>}
 calf_l {parent=thigh_l, loc=<-42.5284 -0.0019 -0.7297>, rot=<0.0006 -0.0161 -0.0391 -0.9991>}
 foot_l {parent=calf_l, loc=<-46.7401 -2.5014 -0.592>, rot=<0.0427 -0.0238 0.0353 -0.9982>}
 ball_l {parent=foot_l, loc=<-3.5658 10.8066 -0.1544>, rot=<0.0001 -0.0001 -0.7063 -0.7079>}
 calf_twist_02_l {parent=calf_l, loc=<-19.1546 -1.7943 -0.2638>, rot=<0.0237 0.0095 -0.0035 -0.9997>}
 calf_twist_01_l {parent=calf_l, loc=<-35.8483 -1.9197 -0.5794>, rot=<0.0237 0.0095 -0.0035 -0.9997>}
 thigh_twist_01_l {parent=thigh_l, loc=<-27.952 0 0>, rot=<0.0013 0 0 -1>}
 thigh_r {parent=pelvis, loc=<-1.0855 -0.2109 8.7053>, rot=<0.0355 -0.0009 0.9993 0.0155>}
 calf_r {parent=thigh_r, loc=<42.5284 0.0019 0.7297>, rot=<0.0006 -0.0161 -0.0391 -0.9991>}
 foot_r {parent=calf_r, loc=<46.7401 2.5014 0.592>, rot=<0.0427 -0.0238 0.0353 -0.9982>}
 ball_r {parent=foot_r, loc=<3.5658 -10.8066 0.1544>, rot=<0.0001 -0.0001 -0.7063 -0.7079>}
 calf_twist_02_r {parent=calf_r, loc=<19.1546 1.7943 0.2638>, rot=<0.0237 0.0095 -0.0035 -0.9997>}
 calf_twist_01_r {parent=calf_r, loc=<35.8483 1.9197 0.5794>, rot=<0.0237 0.0095 -0.0035 -0.9997>}
 thigh_twist_01_r {parent=thigh_r, loc=<27.952 0 0>, rot=<0.0013 0 0 -1>}
]
