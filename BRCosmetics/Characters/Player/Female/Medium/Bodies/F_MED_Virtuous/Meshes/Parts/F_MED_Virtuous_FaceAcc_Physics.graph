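# Physics asset for the face accessory mesh of the female medium Virtuous player body, covering dynamic hair.
# Bang, side and back hair form short capsule chains with tight swing and twist limits, each rooted in a tiny master body constrained to neck_01.
# Spine, neck and clavicle capsules form the upper-body shapes, linked by free constraints.
# The right side hair also carries a second, reversed, fully free constraint between its two segments.
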PhysicsAsset /BRCosmetics/Characters/Player/Female/Medium/Bodies/F_MED_Virtuous/Meshes/Parts/F_MED_Virtuous_FaceAcc_Physics
bodies {
 clavicle_l [{shapeType=Capsule, length=8, radius=7, shapeName=clavicle_l_capsule, center=<8.0946 -0.4269 -3.0858>, rotation=<rot 90 0 0>}]
 clavicle_r [{shapeType=Capsule, length=8, radius=7, shapeName=clavicle_r_capsule, center=<-8.0946 0.4269 3.0858>, rotation=<rot 90 0 0>}]
 dyn_bang_l_01 [{shapeType=Capsule, length=3, radius=2.5, shapeName=dyn_bang_l_01_capsule, center=<1 0 0>, rotation=<rot 90 0 0>}]
 dyn_bang_l_02 [{shapeType=Capsule, length=3, radius=2.5, shapeName=dyn_bang_l_02_capsule, center=<1 0 0>, rotation=<rot 90 0 0>}]
 dyn_bang_l_MSTR [{shapeType=Capsule, length=0.1, radius=0.1, shapeName=dyn_bang_l_MSTR_capsule}]
 dyn_bang_r_01 [{shapeType=Capsule, length=3, radius=2.5, shapeName=dyn_bang_r_01_capsule, center=<1 0 0>, rotation=<rot 90 0 0>}]
 dyn_bang_r_02 [{shapeType=Capsule, length=3, radius=2.5, shapeName=dyn_bang_r_02_capsule, center=<1 0 0>, rotation=<rot 90 0 0>}]
 dyn_bang_r_MSTR [{shapeType=Capsule, length=0.1, radius=0.1, shapeName=dyn_bang_r_MSTR_capsule}]
 dyn_hair_back_01 [{shapeType=Capsule, length=13, radius=5, shapeName=dyn_hair_back_01_capsule, center=<3 0 2.5>, rotation=<rot 0 0 90>}]
 dyn_hair_back_02 [{shapeType=Capsule, length=14.75, radius=5, shapeName=dyn_hair_back_02_capsule, center=<3 0 2.5>, rotation=<rot 0 0 90>}]
 dyn_hair_back_03 [{shapeType=Capsule, length=15.75, radius=5, shapeName=dyn_hair_back_03_capsule, center=<3 0 2.5>, rotation=<rot 0 0 90>}]
 dyn_hair_back_MSTR [{shapeType=Sphere, radius=0.505, shapeName=dyn_hair_back_MSTR_sphere}]
 dyn_hair_l_01 [{shapeType=Capsule, length=8.75, radius=4, shapeName=dyn_hair_back_02_capsule, center=<2 0 -1>, rotation=<rot 0 25 90>}]
 dyn_hair_l_02 [{shapeType=Capsule, length=8.75, radius=4, shapeName=dyn_hair_back_02_capsule, center=<2 0 -1>, rotation=<rot 0 25 90>}]
 dyn_hair_l_MSTR [{shapeType=Sphere, radius=0.505, shapeName=dyn_hair_l_MSTR_sphere}]
 dyn_hair_r_01 [{shapeType=Capsule, length=9.75, radius=3, shapeName=dyn_hair_back_02_capsule, center=<2.5218 0 -1>, rotation=<rot 0 -20 90>}]
 dyn_hair_r_02 [{shapeType=Capsule, length=9.75, radius=3, shapeName=dyn_hair_back_02_capsule, center=<2.5218 0 -1>, rotation=<rot 0 -20 90>}]
 dyn_hair_r_MSTR [{shapeType=Sphere, radius=0.505, shapeName=dyn_hair_r_MSTR_sphere}]
 neck_01 [{shapeType=Capsule, length=4, radius=4, shapeName=neck_01_capsule, rotation=<rot 90 0 0>}]
 spine_04 [{shapeType=Capsule, length=12, radius=7, shapeName=spine_05_capsule, center=<1.8111 -0.2588 0>}]
 spine_05 [{shapeType=Capsule, length=12, radius=7, shapeName=spine_05_capsule, center=<1.8086 -0.259 0>}]
}
constraints [
 {bone1Name=dyn_bang_l_01, bone2Name=dyn_bang_l_MSTR, swing1LimitDegrees=5, swing2LimitDegrees=18, twistLimitDegrees=2}
 {bone1Name=dyn_bang_l_02, bone2Name=dyn_bang_l_01, swing1LimitDegrees=5, swing2LimitDegrees=18, twistLimitDegrees=2}
 {bone1Name=dyn_bang_r_01, bone2Name=dyn_bang_r_MSTR, swing1LimitDegrees=5, swing2LimitDegrees=18, twistLimitDegrees=2}
 {bone1Name=dyn_bang_r_02, bone2Name=dyn_bang_r_01, swing1LimitDegrees=5, swing2LimitDegrees=18, twistLimitDegrees=2}
 {
  bone1Name=dyn_bang_r_MSTR
  bone2Name=neck_01
  swing1LimitDegrees=45
  swing1Motion=Free
  swing2LimitDegrees=45
  swing2Motion=Free
  twistLimitDegrees=45
  twistMotion=Free
 }
 {
  bone1Name=dyn_bang_l_MSTR
  bone2Name=neck_01
  swing1LimitDegrees=45
  swing1Motion=Free
  swing2LimitDegrees=45
  swing2Motion=Free
  twistLimitDegrees=45
  twistMotion=Free
 }
 {bone1Name=dyn_hair_back_03, bone2Name=dyn_hair_back_02, swing1LimitDegrees=6, swing2LimitDegrees=24, twistLimitDegrees=30}
 {
  bone1Name=dyn_hair_back_MSTR
  bone2Name=neck_01
  swing1LimitDegrees=45
  swing1Motion=Free
  swing2LimitDegrees=45
  swing2Motion=Free
  twistLimitDegrees=45
  twistMotion=Free
 }
 {bone1Name=dyn_hair_back_01, bone2Name=dyn_hair_back_MSTR, swing1LimitDegrees=6, swing2LimitDegrees=24, twistLimitDegrees=30}
 {bone1Name=dyn_hair_back_02, bone2Name=dyn_hair_back_01, swing1LimitDegrees=6, swing2LimitDegrees=24, twistLimitDegrees=30}
 {
  bone1Name=clavicle_l
  bone2Name=spine_05
  swing1LimitDegrees=45
  swing1Motion=Free
  swing2LimitDegrees=45
  swing2Motion=Free
  twistLimitDegrees=45
  twistMotion=Free
 }
 {
  bone1Name=clavicle_r
  bone2Name=spine_05
  swing1LimitDegrees=45
  swing1Motion=Free
  swing2LimitDegrees=45
  swing2Motion=Free
  twistLimitDegrees=45
  twistMotion=Free
 }
 {
  bone1Name=dyn_hair_l_MSTR
  bone2Name=neck_01
  swing1LimitDegrees=45
  swing1Motion=Free
  swing2LimitDegrees=45
  swing2Motion=Free
  twistLimitDegrees=45
  twistMotion=Free
 }
 {
  bone1Name=dyn_hair_r_MSTR
  bone2Name=neck_01
  swing1LimitDegrees=45
  swing1Motion=Free
  swing2LimitDegrees=45
  swing2Motion=Free
  twistLimitDegrees=45
  twistMotion=Free
 }
 {bone1Name=dyn_hair_l_01, bone2Name=dyn_hair_l_MSTR, swing1LimitDegrees=14, swing2LimitDegrees=12, twistLimitDegrees=12}
 {bone1Name=dyn_hair_l_02, bone2Name=dyn_hair_l_01, swing1LimitDegrees=14, swing2LimitDegrees=12, twistLimitDegrees=2}
 {bone1Name=dyn_hair_r_01, bone2Name=dyn_hair_r_MSTR, swing1LimitDegrees=24, swing2LimitDegrees=12, twistLimitDegrees=12}
 {bone1Name=dyn_hair_r_02, bone2Name=dyn_hair_r_01, swing1LimitDegrees=24, swing2LimitDegrees=12, twistLimitDegrees=2}
 {
  bone1Name=dyn_hair_r_01
  bone2Name=dyn_hair_r_02
  swing1LimitDegrees=45
  swing1Motion=Free
  swing2LimitDegrees=45
  swing2Motion=Free
  twistLimitDegrees=45
  twistMotion=Free
 }
 {
  bone1Name=neck_01
  bone2Name=spine_05
  swing1LimitDegrees=45
  swing1Motion=Free
  swing2LimitDegrees=45
  swing2Motion=Free
  twistLimitDegrees=45
  twistMotion=Free
 }
 {
  bone1Name=spine_05
  bone2Name=spine_04
  swing1LimitDegrees=45
  swing1Motion=Free
  swing2LimitDegrees=45
  swing2Motion=Free
  twistLimitDegrees=45
  twistMotion=Free
 }
]
