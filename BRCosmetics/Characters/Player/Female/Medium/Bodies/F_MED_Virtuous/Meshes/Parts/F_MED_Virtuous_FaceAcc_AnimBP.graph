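# Post-copy animation Blueprint for the face accessory mesh of the F_MED_Virtuous character body, adding secondary motion to it.
# It copies the pose from the attached parent mesh and caches it, then runs the accessory Control Rig into a RigidBody simulation.
# The simulation works in spine_05 bone space with scaled, clamped component motion.
# bRigidBodyEnabled switches between the simulated result and the plain copied pose, bypassing both rig and physics.
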AnimBlueprint /BRCosmetics/Characters/Player/Female/Medium/Bodies/F_MED_Virtuous/Meshes/Parts/F_MED_Virtuous_FaceAcc_AnimBP
target_skeleton=&/F_MED_Virtuous_FaceAcc_Skeleton
graphs [
 AnimGraph {nodes [
   n1 = Control Rig(Source=Use cached pose 'SavedPose10302'()) {
    alphaBoolBlend {blendOption=Linear}
    controlRigAssetReference {blueprintRigClass=&/F_MED_Virtuous_FaceAcc_CtrlRig.F_MED_Virtuous_FaceAcc_CtrlRig_C}
   }
   n2 = RigidBody(ComponentPose=Local To Component(LocalPose=@n1)) {
    overridePhysicsAsset=&/F_MED_Virtuous_FaceAcc_Physics
    simulationSpace=BaseBoneSpace
    alphaBoolBlend {blendOption=Linear}
    baseBoneRef {boneName=spine_05}
    componentAppliedLinearAccClamp=<3000 3000 3000>
    componentLinearAccScale=<0.2 0.2 0.4>
    componentLinearVelScale=<0.1 0.1 0.2>
   }
   Output Pose(Result=Blend Poses by bool(BlendPose_0=Component To Local(ComponentPose=@n2), BlendPose_1=Use cached pose 'SavedPose10302'(), bActiveValue=Get bRigidBodyEnabled()) {blendType=Linear}) {layerGroup=None}
   Save cached pose 'SavedPose10302'(Pose=Copy Pose From Mesh() {bUseAttachedParent=true}) {cacheName=SavedPose10302}
  ]}
 EventGraph {nodes [Event BlueprintUpdateAnimation()]}
]
variables []
